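# Physics asset for the frontend mesh of the F_MED_Painter female body, set up around a long, bone-driven jacket.
# Capsules cover pelvis, spine, legs and feet, while most jacket bones use two or three parallel capsules to approximate flat panels.
# Front, front-side, side and back jacket chains hang from spine_03 with tight twist limits.
# Looser cross-constraints tie neighboring chains into a cloth-like lattice; a few joints have free angular motion.
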PhysicsAsset /BRCosmetics/Characters/Player/Female/Medium/Bodies/F_MED_Painter/Meshes/F_MED_Painter_Frontend_Phys
bodies {
 calf_l [{shapeType=Capsule, length=49.722, radius=7.5, center=<-23.2226 -4.6248 0>, rotation=<rot 90 0 175>}]
 calf_r [{shapeType=Capsule, length=43.7818, radius=5.9594, center=<19.1947 3.587 0.0921>, rotation=<rot 87.5184 -130.032 -134.5437>}]
 foot_l [{shapeType=Capsule, length=12.6052, radius=6.5119, center=<-2.9081 7.2701 -0.3337>, rotation=<rot 0 5.1452 90.9165>}]
 foot_r [{shapeType=Capsule, length=12.6052, radius=6.5119, center=<2.9081 -7.2701 0.3337>, rotation=<rot 0 5.1452 -89.0835>}]
 jacket_back_01_l [
  {
   shapeType=Capsule
   length=13.8864
   radius=2.6813
   shapeName=jacket_back_01_l_capsule
   center=<0.0176 0 -2.5186>
   rotation=<rot 86.8694 0.0004 179.9996>
  }
  {
   shapeType=Capsule
   length=13.8864
   radius=2.6813
   shapeName=jacket_back_01_l_capsule
   center=<0.0816 0 2.7259>
   rotation=<rot 86.8694 0.0004 179.9996>
  }
 ]
 jacket_back_01_r [
  {
   shapeType=Capsule
   length=13.8864
   radius=2.6813
   shapeName=jacket_back_01_l_capsule
   center=<-0.0176 0 2.5186>
   rotation=<rot 86.8694 0.0004 -0.0004>
  }
  {
   shapeType=Capsule
   length=13.8864
   radius=2.6813
   shapeName=jacket_back_01_l_capsule
   center=<-0.0816 0 -2.7259>
   rotation=<rot 86.8694 0.0004 -0.0004>
  }
 ]
 jacket_back_02 [
  {
   shapeType=Capsule
   length=21.0401
   radius=3.1386
   shapeName=jacket_back_02_capsule
   center=<-3.9454 0 5.7823>
   rotation=<rot -90 179.9926 -179.9925>
  }
  {
   shapeType=Capsule
   length=21.0401
   radius=3.1386
   shapeName=jacket_back_02_capsule
   center=<-3.9454 0 0.1167>
   rotation=<rot -90 180 -179.9998>
  }
  {
   shapeType=Capsule
   length=21.0401
   radius=3.1386
   shapeName=jacket_back_02_capsule
   center=<-3.9454 0 -5.6091>
   rotation=<rot -90 180 -179.9998>
  }
 ]
 jacket_back_02_l [
  {
   shapeType=Capsule
   length=18.9967
   radius=3.5779
   shapeName=jacket_back_02_l_capsule
   center=<-3.7931 -0.3163 5.0031>
   rotation=<rot 85.7638 -89.9973 -89.1793>
  }
  {
   shapeType=Capsule
   length=18.9967
   radius=3.5779
   shapeName=jacket_back_02_l_capsule
   center=<-3.7931 0.0705 -0.2186>
   rotation=<rot 85.7638 -89.9973 -89.1793>
  }
  {
   shapeType=Capsule
   length=18.9967
   radius=3.5779
   shapeName=jacket_back_02_l_capsule
   center=<-3.7931 0.4092 -4.7907>
   rotation=<rot 85.7638 -89.9973 -89.1793>
  }
 ]
 jacket_back_02_r [
  {
   shapeType=Capsule
   length=18.9967
   radius=3.5779
   shapeName=jacket_back_02_l_capsule
   center=<-3.7931 -0.3163 5.0031>
   rotation=<rot 85.7638 -89.9973 -89.1793>
  }
  {
   shapeType=Capsule
   length=18.9967
   radius=3.5779
   shapeName=jacket_back_02_l_capsule
   center=<-3.7931 0.0705 -0.2186>
   rotation=<rot 85.7638 -89.9973 -89.1793>
  }
  {
   shapeType=Capsule
   length=18.9967
   radius=3.5779
   shapeName=jacket_back_02_l_capsule
   center=<-3.7931 0.4092 -4.7907>
   rotation=<rot 85.7638 -89.9973 -89.1793>
  }
 ]
 jacket_back_03 [
  {
   shapeType=Capsule
   length=20.6855
   radius=4.3135
   shapeName=jacket_back_03_capsule
   center=<-5.9921 0 4.6527>
   rotation=<rot 90 6.7213 6.7211>
  }
  {
   shapeType=Capsule
   length=20.6855
   radius=4.3135
   shapeName=jacket_back_03_capsule
   center=<-5.9921 0 -4.139>
   rotation=<rot 90 0.0001 -0.0001>
  }
 ]
 jacket_back_03_l [
  {
   shapeType=Capsule
   length=18.4989
   radius=4.2963
   shapeName=jacket_back_03_l_capsule
   center=<-6.0739 0.542 4.6162>
   rotation=<rot 88.6954 179.999 179.9984>
  }
  {
   shapeType=Capsule
   length=18.4989
   radius=4.2963
   shapeName=jacket_back_03_l_capsule
   center=<-5.8701 0.542 -4.3321>
   rotation=<rot 88.6954 179.999 179.9984>
  }
 ]
 jacket_back_03_r [
  {
   shapeType=Capsule
   length=18.4989
   radius=4.2963
   shapeName=jacket_back_03_l_capsule
   center=<-6.1015 0.542 4.0436>
   rotation=<rot 87.4106 0.0009 -0.0003>
  }
  {
   shapeType=Capsule
   length=18.4989
   radius=4.2963
   shapeName=jacket_back_03_l_capsule
   center=<-6.4563 0.542 -4.3027>
   rotation=<rot 87.133 0.0012 0.0003>
  }
 ]
 jacket_back_04 [
  {shapeType=Capsule, length=16.8919, radius=3.5779, shapeName=jacket_back_04_capsule, center=<-0.5964 -0.3246 0>}
  {shapeType=Capsule, length=16.8919, radius=3.5779, shapeName=jacket_back_04_capsule, center=<-7.7448 -0.3246 0>}
  {shapeType=Capsule, length=16.8919, radius=3.5779, shapeName=jacket_back_04_capsule, center=<-14.8807 -0.3246 0>}
 ]
 jacket_back_04_l [
  {
   shapeType=Capsule
   length=17.8218
   radius=4.1761
   shapeName=jacket_back_04_l_capsule
   center=<-7.0653 -0.3371 4.4877>
   rotation=<rot 83.1913 -89.9962 -93.2433>
  }
  {
   shapeType=Capsule
   length=17.8218
   radius=4.1761
   shapeName=jacket_back_04_l_capsule
   center=<-7.0654 0.6652 -3.9077>
   rotation=<rot 83.1913 -89.9962 -93.2433>
  }
 ]
 jacket_back_04_r [
  {
   shapeType=Capsule
   length=17.8218
   radius=4.1761
   shapeName=jacket_back_04_l_capsule
   center=<-7.0218 -0.3762 4.1601>
   rotation=<rot 82.0783 68.5131 65.5446>
  }
  {
   shapeType=Capsule
   length=17.8218
   radius=4.1761
   shapeName=jacket_back_04_l_capsule
   center=<-7.3894 -1.4766 -3.9479>
   rotation=<rot 82.0783 68.5131 65.5446>
  }
 ]
 jacket_back_05 [
  {shapeType=Capsule, length=17.2323, radius=3.6375, shapeName=jacket_back_05_capsule, center=<-0.1912 -0.8778 0>}
  {shapeType=Capsule, length=17.2323, radius=3.6375, shapeName=jacket_back_05_capsule, center=<-10.7438 -0.8778 0>}
 ]
 jacket_back_05_l [
  {
   shapeType=Capsule
   length=16.4775
   radius=3.0856
   shapeName=jacket_back_05_l_capsule
   center=<1.569 -0.8297 0.2392>
   rotation=<rot 1.9039 0.7154 -4.6112>
  }
  {
   shapeType=Capsule
   length=16.4775
   radius=3.0856
   shapeName=jacket_back_05_l_capsule
   center=<-5.1172 -0.9132 0.017>
   rotation=<rot 1.9039 0.7154 -4.6112>
  }
  {
   shapeType=Capsule
   length=16.4775
   radius=3.0856
   shapeName=jacket_back_05_l_capsule
   center=<-11.7586 -0.9962 -0.2038>
   rotation=<rot 1.9039 0.7154 -4.6112>
  }
 ]
 jacket_back_05_r [
  {
   shapeType=Capsule
   length=15.7975
   radius=3.0856
   shapeName=jacket_back_05_l_capsule
   center=<1.5449 -0.2487 -0.4292>
   rotation=<rot -1.9214 0.1662 8.1829>
  }
  {
   shapeType=Capsule
   length=15.7975
   radius=3.0856
   shapeName=jacket_back_05_l_capsule
   center=<-5.1224 -0.2497 -0.0784>
   rotation=<rot -1.9214 0.1662 8.1829>
  }
  {
   shapeType=Capsule
   length=15.7975
   radius=3.0856
   shapeName=jacket_back_05_l_capsule
   center=<-11.7585 -0.3102 -0.1435>
   rotation=<rot -1.9214 0.1662 8.1829>
  }
 ]
 jacket_frtSide_03_l [
  {shapeType=Capsule, length=15.5437, radius=4.0191, shapeName=jacket_frtSide_03_l_capsule, center=<-6.1485 2 -3.7202>, rotation=<rot 90 0 0>}
  {shapeType=Capsule, length=15.5437, radius=4.0191, shapeName=jacket_frtSide_03_l_capsule, center=<-6.1485 2 3.7794>, rotation=<rot 90 0 0>}
 ]
 jacket_frtSide_03_r [
  {
   shapeType=Capsule
   length=15.5437
   radius=4.0191
   shapeName=jacket_frtSide_03_l_capsule
   center=<-6.1485 3 -3.7202>
   rotation=<rot 90 -180 -180>
  }
  {
   shapeType=Capsule
   length=15.5437
   radius=4.0191
   shapeName=jacket_frtSide_03_l_capsule
   center=<-6.1485 3 3.7794>
   rotation=<rot 90 -180 -180>
  }
 ]
 jacket_frtSide_04_l [
  {shapeType=Capsule, length=13.8569, radius=3.6941, shapeName=jacket_frtSide_04_l_capsule, center=<-3.5177 2 0>}
  {shapeType=Capsule, length=13.8569, radius=3.6941, shapeName=jacket_frtSide_04_l_capsule, center=<-10.7264 2 0>}
 ]
 jacket_frtSide_04_r [
  {shapeType=Capsule, length=13.8569, radius=3.6941, shapeName=jacket_frtSide_04_l_capsule, center=<-3.5177 3 0>}
  {shapeType=Capsule, length=13.8569, radius=3.6941, shapeName=jacket_frtSide_04_l_capsule, center=<-10.7264 3 0>}
 ]
 jacket_frtSide_05_l [
  {shapeType=Capsule, length=14.1901, radius=2.9446, shapeName=jacket_frtSide_05_l_capsule, center=<-2.3462 1 0>}
  {shapeType=Capsule, length=14.1901, radius=2.9446, shapeName=jacket_frtSide_05_l_capsule, center=<-7.1593 1 0>}
 ]
 jacket_frtSide_05_r [
  {shapeType=Capsule, length=14.1901, radius=2.9446, shapeName=jacket_frtSide_05_l_capsule, center=<-2.3462 1 0>}
  {shapeType=Capsule, length=14.1901, radius=2.9446, shapeName=jacket_frtSide_05_l_capsule, center=<-7.1593 1 0>}
 ]
 jacket_frt_03_l [
  {
   shapeType=Capsule
   length=15.1395
   radius=3.4102
   shapeName=jacket_frt_03_l_capsule
   center=<-4.8809 1 2.826>
   rotation=<rot 88.3115 -176.4516 -179.136>
  }
  {
   shapeType=Capsule
   length=15.1395
   radius=3.4102
   shapeName=jacket_frt_03_l_capsule
   center=<-4.723 1 -2.5414>
   rotation=<rot 88.3115 -176.4516 -179.136>
  }
 ]
 jacket_frt_03_r [
  {
   shapeType=Capsule
   length=15.1395
   radius=3.4102
   shapeName=jacket_frt_03_l_capsule
   center=<-4.8809 1 2.826>
   rotation=<rot 88.3115 -176.4516 -179.136>
  }
  {
   shapeType=Capsule
   length=15.1395
   radius=3.4102
   shapeName=jacket_frt_03_l_capsule
   center=<-4.723 1 -2.5414>
   rotation=<rot 88.3115 -176.4516 -179.136>
  }
 ]
 jacket_frt_04_l [
  {
   shapeType=Capsule
   length=17.0347
   radius=3.2733
   shapeName=jacket_frt_04_l_capsule
   center=<-6.9872 2.5 2.5817>
   rotation=<rot 88.3065 0.2561 -0.2439>
  }
  {
   shapeType=Capsule
   length=17.0347
   radius=3.2733
   shapeName=jacket_frt_04_l_capsule
   center=<-7.14 2.5 -2.5863>
   rotation=<rot 88.3065 0.2561 -0.2439>
  }
 ]
 jacket_frt_04_r [
  {
   shapeType=Capsule
   length=15.1597
   radius=3.2733
   shapeName=jacket_frt_04_l_capsule
   center=<-7.0152 2.5 2.5821>
   rotation=<rot 88.3065 0.2561 -0.2439>
  }
  {
   shapeType=Capsule
   length=15.1597
   radius=3.2733
   shapeName=jacket_frt_04_l_capsule
   center=<-7.168 2.5 -2.5859>
   rotation=<rot 88.3065 0.2561 -0.2439>
  }
 ]
 jacket_frt_05_l [
  {shapeType=Capsule, length=13.1207, radius=3.41, shapeName=jacket_frt_05_l_capsule, center=<-6.5059 2 2.2757>, rotation=<rot 90 180 177.3533>}
  {shapeType=Capsule, length=13.1207, radius=3.41, shapeName=jacket_frt_05_l_capsule, center=<-6.5059 2 -2.396>, rotation=<rot 90 180 177.3533>}
 ]
 jacket_frt_05_r [
  {shapeType=Capsule, length=13.1207, radius=3.41, shapeName=jacket_frt_05_l_capsule, center=<-6.5059 2 2.2757>, rotation=<rot 90 180 177.3533>}
  {shapeType=Capsule, length=13.1207, radius=3.41, shapeName=jacket_frt_05_l_capsule, center=<-6.5059 2 -2.396>, rotation=<rot 90 180 177.3533>}
 ]
 jacket_frt_06_l [
  {
   shapeType=Capsule
   length=13.9816
   radius=3.191
   shapeName=jacket_frt_06_l_capsule
   center=<-4.1853 1 2.4826>
   rotation=<rot 89.7173 0.9298 0.0012>
  }
  {
   shapeType=Capsule
   length=13.9816
   radius=3.191
   shapeName=jacket_frt_06_l_capsule
   center=<-4.2081 1 -2.1357>
   rotation=<rot 89.7173 0.9298 0.0012>
  }
 ]
 jacket_frt_06_r [
  {
   shapeType=Capsule
   length=13.9816
   radius=3.191
   shapeName=jacket_frt_06_l_capsule
   center=<-4.1853 1 2.4826>
   rotation=<rot 89.7173 0.9298 0.0012>
  }
  {
   shapeType=Capsule
   length=13.9816
   radius=3.191
   shapeName=jacket_frt_06_l_capsule
   center=<-4.2081 1 -2.1357>
   rotation=<rot 89.7173 0.9298 0.0012>
  }
 ]
 jacket_frt_07_l [{shapeType=Capsule, length=6.2952, radius=2.3579, shapeName=jacket_frt_07_l_capsule, center=<-1.5968 0 0>, rotation=<rot -90 0 0>}]
 jacket_frt_07_r [{shapeType=Capsule, length=6.2952, radius=2.3579, shapeName=jacket_frt_07_l_capsule, center=<-1.5968 0 0>, rotation=<rot -90 0 0>}]
 jacket_frt_08_l [{shapeType=Capsule, length=3.3982, radius=1.7525, shapeName=jacket_frt_08_l_capsule, center=<-2.1344 0 0>, rotation=<rot 90 90 90>}]
 jacket_frt_08_r [{shapeType=Capsule, length=3.3982, radius=1.7525, shapeName=jacket_frt_08_l_capsule, center=<-2.2254 0 0>, rotation=<rot 90 3.3723 -176.6277>}]
 jacket_frt_09_l [{
   shapeType=Capsule
   length=2.9121
   radius=1.8845
   shapeName=jacket_frt_09_l_capsule
   center=<-2.607 0 0>
   rotation=<rot -90 -179.9604 179.9605>
  }]
 jacket_frt_09_r [{
   shapeType=Capsule
   length=2.9121
   radius=1.8845
   shapeName=jacket_frt_09_l_capsule
   center=<-2.6977 0 0>
   rotation=<rot -90 -179.9604 -0.0395>
  }]
 jacket_side_01_l [
  {
   shapeType=Capsule
   length=7.0001
   radius=2.6072
   shapeName=jacket_side_01_l_capsule
   center=<-1.4649 -0.0001 -0.0471>
   rotation=<rot -3.4338 0.0044 -179.9998>
  }
  {
   shapeType=Capsule
   length=7.0001
   radius=2.6072
   shapeName=jacket_side_01_l_capsule
   center=<1.8306 0.0001 0.0404>
   rotation=<rot -3.4338 0.0044 -179.9998>
  }
 ]
 jacket_side_01_r [
  {
   shapeType=Capsule
   length=7.0001
   radius=2.6072
   shapeName=jacket_side_01_l_capsule
   center=<1.4649 0.0001 0.0471>
   rotation=<rot -3.4338 0.0044 0.0002>
  }
  {
   shapeType=Capsule
   length=7.0001
   radius=2.6072
   shapeName=jacket_side_01_l_capsule
   center=<-1.8306 -0.0001 -0.0404>
   rotation=<rot -3.4338 0.0044 0.0002>
  }
 ]
 jacket_side_02_l [
  {
   shapeType=Capsule
   length=14.0407
   radius=4.1052
   shapeName=jacket_side_02_l_capsule
   center=<-0.6429 2.1794 0.5491>
   rotation=<rot -1.3372 3.7596 -176.5172>
  }
  {
   shapeType=Capsule
   length=14.0407
   radius=4.1052
   shapeName=jacket_side_02_l_capsule
   center=<-8.7132 2.1794 0.3603>
   rotation=<rot -1.3372 3.7596 -176.5172>
  }
 ]
 jacket_side_02_r [
  {
   shapeType=Capsule
   length=14.0407
   radius=4.1052
   shapeName=jacket_side_02_l_capsule
   center=<-0.7755 0 -0.0944>
   rotation=<rot -1.3371 3.7595 -11.7694>
  }
  {
   shapeType=Capsule
   length=14.0407
   radius=4.1052
   shapeName=jacket_side_02_l_capsule
   center=<-8.8457 0 0.0945>
   rotation=<rot -1.3371 3.7595 -11.7694>
  }
 ]
 jacket_side_03_l [
  {
   shapeType=Capsule
   length=14.783
   radius=3.6633
   shapeName=jacket_side_03_l_capsule
   center=<-0.4366 2 -0.0344>
   rotation=<rot -0.9903 1.7003 9.9759>
  }
  {
   shapeType=Capsule
   length=14.783
   radius=3.6633
   shapeName=jacket_side_03_l_capsule
   center=<-6.4118 2 0.0689>
   rotation=<rot -0.9903 1.7003 9.9759>
  }
  {
   shapeType=Capsule
   length=14.783
   radius=3.6633
   shapeName=jacket_side_03_l_capsule
   center=<-12.096 2 0.1672>
   rotation=<rot -0.9903 1.7003 9.9759>
  }
 ]
 jacket_side_03_r [
  {
   shapeType=Capsule
   length=13.9644
   radius=3.6633
   shapeName=jacket_side_03_l_capsule
   center=<-0.4366 1 -0.0344>
   rotation=<rot 1.1 1.3845 -8.6027>
  }
  {
   shapeType=Capsule
   length=13.9644
   radius=3.6633
   shapeName=jacket_side_03_l_capsule
   center=<-6.4118 1 0.0689>
   rotation=<rot 1.1 1.3845 -8.6027>
  }
  {
   shapeType=Capsule
   length=13.9644
   radius=3.6633
   shapeName=jacket_side_03_l_capsule
   center=<-12.1851 1 0.6654>
   rotation=<rot 1.1 1.3845 -8.6027>
  }
 ]
 jacket_side_04_l [
  {
   shapeType=Capsule
   length=13.4633
   radius=3.4732
   shapeName=jacket_side_04_l_capsule
   center=<-4.1304 1.5 0.0042>
   rotation=<rot -1.3806 -0.2385 9.8085>
  }
  {
   shapeType=Capsule
   length=13.4633
   radius=3.4732
   shapeName=jacket_side_04_l_capsule
   center=<-10.7105 1.5 0.1628>
   rotation=<rot -1.3806 -0.2385 9.8085>
  }
 ]
 jacket_side_04_r [
  {
   shapeType=Capsule
   length=12.7219
   radius=3.4732
   shapeName=jacket_side_04_l_capsule
   center=<-3.7194 2.5 0.5925>
   rotation=<rot 0.4832 -0.5869 -10.5941>
  }
  {
   shapeType=Capsule
   length=12.7219
   radius=3.4732
   shapeName=jacket_side_04_l_capsule
   center=<-10.3094 1.5 0.5363>
   rotation=<rot 0.4832 -0.5869 -10.5941>
  }
 ]
 jacket_side_05_l [
  {
   shapeType=Capsule
   length=13.1461
   radius=3.1707
   shapeName=jacket_side_05_l_capsule
   center=<-2.9833 1 0.1145>
   rotation=<rot -0.7136 -0.1012 8.0971>
  }
  {
   shapeType=Capsule
   length=13.1461
   radius=3.1707
   shapeName=jacket_side_05_l_capsule
   center=<-6.7375 1 0.1612>
   rotation=<rot -0.7136 -0.1012 8.0971>
  }
 ]
 jacket_side_05_r [
  {
   shapeType=Capsule
   length=13.1162
   radius=3.1707
   shapeName=jacket_side_05_l_capsule
   center=<-2.9801 1 0.0824>
   rotation=<rot -0.7134 -0.1012 -9.9759>
  }
  {
   shapeType=Capsule
   length=13.1162
   radius=3.1707
   shapeName=jacket_side_05_l_capsule
   center=<-6.7343 1 0.1292>
   rotation=<rot -0.7134 -0.1012 -9.9759>
  }
 ]
 jacket_side_06_l [
  {
   shapeType=Capsule
   length=1.759
   radius=2.9657
   shapeName=jacket_side_06_l_capsule
   center=<0 -0.1418 -0.0098>
   rotation=<rot -0.7805 0.0586 -3.9666>
  }
  {
   shapeType=Capsule
   length=1.759
   radius=2.9657
   shapeName=jacket_side_06_l_capsule
   center=<-3.3961 -0.1452 0.0364>
   rotation=<rot -0.7805 0.0586 -3.9666>
  }
 ]
 jacket_side_06_r [
  {
   shapeType=Capsule
   length=1.759
   radius=2.9657
   shapeName=jacket_side_06_l_capsule
   center=<0 -0.1418 -0.0098>
   rotation=<rot -0.7805 0.0586 -3.9666>
  }
  {
   shapeType=Capsule
   length=1.759
   radius=2.9657
   shapeName=jacket_side_06_l_capsule
   center=<-3.3961 -0.1452 0.0364>
   rotation=<rot -0.7805 0.0586 -3.9666>
  }
 ]
 jacket_side_07_l [
  {shapeType=Capsule, length=1.9995, radius=2.6742, shapeName=jacket_side_07_l_capsule, center=<-1.1853 0 -0.3722>}
  {shapeType=Capsule, length=1.9995, radius=2.6742, shapeName=jacket_side_07_l_capsule, center=<-3.5532 0 -0.3722>}
 ]
 jacket_side_07_r [
  {shapeType=Capsule, length=1.9995, radius=2.6742, shapeName=jacket_side_07_l_capsule, center=<-1.1853 0 -0.3722>}
  {shapeType=Capsule, length=1.9995, radius=2.6742, shapeName=jacket_side_07_l_capsule, center=<-3.5532 0 -0.3722>}
 ]
 jacket_side_08_l [
  {
   shapeType=Capsule
   length=2.7333
   radius=2.0166
   shapeName=jacket_side_08_l_capsule
   center=<-1.3969 0.0631 -0.2672>
   rotation=<rot 1.4216 -2.0746 -101.8184>
  }
  {
   shapeType=Capsule
   length=2.7333
   radius=2.0166
   shapeName=jacket_side_08_l_capsule
   center=<-3.8293 0.1512 -0.3276>
   rotation=<rot 1.4216 -2.0746 -101.8184>
  }
 ]
 jacket_side_08_r [
  {
   shapeType=Capsule
   length=2.7333
   radius=2.0166
   shapeName=jacket_side_08_l_capsule
   center=<-1.3969 0.0631 -0.2672>
   rotation=<rot 1.4216 -2.0746 -80.1806>
  }
  {
   shapeType=Capsule
   length=2.7333
   radius=2.0166
   shapeName=jacket_side_08_l_capsule
   center=<-3.8293 0.1512 -0.3276>
   rotation=<rot 1.4216 -2.0746 -80.1806>
  }
 ]
 pelvis [{shapeType=Capsule, length=6.7403, radius=11.0058, center=<-1.9341 -0.9281 0.091>}]
 spine_01 [{shapeType=Capsule, length=6.4716, radius=9.2319, shapeName=spine_01_capsule, center=<0.8698 0 0>, rotation=<rot 0 -90 0>}]
 spine_03 [{shapeType=Capsule, length=7.2052, radius=11.0121, center=<-4.3339 3.3752 -0.3231>, rotation=<rot 6.2674 -80.8928 90.3065>}]
 thigh_l [
  {shapeType=Capsule, length=8, radius=7.5, center=<-20.9471 -0.0908 0.161>, rotation=<rot 87.6152 -12.0017 165.9154>}
  {shapeType=Capsule, length=5.5, radius=6.5, center=<-31.9914 -1.0011 0.2198>, rotation=<rot -84.9731 -176.8868 2.0178>}
  {shapeType=Capsule, length=11.0096, radius=9.3994, center=<-5.195 -2.637 2.4536>, rotation=<rot 82.6484 170.8883 -1.0782>}
 ]
 thigh_r [
  {shapeType=Capsule, length=8, radius=7.5, center=<20.9471 0.0908 -0.161>, rotation=<rot 87.6152 -12.0017 -14.0846>}
  {shapeType=Capsule, length=5.5, radius=6.5, center=<31.9914 1.0011 -0.2198>, rotation=<rot -84.9731 -176.8868 -177.9822>}
  {shapeType=Capsule, length=11.0096, radius=9.3994, center=<5.195 2.637 -2.4536>, rotation=<rot 82.6484 170.8883 178.9218>}
 ]
}
constraints [
 {bone1Name=spine_03, bone2Name=spine_01, swing1LimitDegrees=45, swing2LimitDegrees=45, twistLimitDegrees=45}
 {bone1Name=calf_r, bone2Name=thigh_r, swing1LimitDegrees=45, swing2LimitDegrees=45, twistLimitDegrees=45}
 {
  bone1Name=foot_r
  bone2Name=calf_r
  swing1LimitDegrees=45
  swing1Motion=Free
  swing2LimitDegrees=45
  swing2Motion=Free
  twistLimitDegrees=45
  twistMotion=Free
 }
 {bone1Name=jacket_frt_03_r, bone2Name=spine_03, swing1LimitDegrees=80, swing2LimitDegrees=60, twistLimitDegrees=8}
 {bone1Name=jacket_frt_04_r, bone2Name=jacket_frt_03_r, swing1LimitDegrees=80, swing2LimitDegrees=60, twistLimitDegrees=8}
 {bone1Name=jacket_frt_05_r, bone2Name=jacket_frt_04_r, swing1LimitDegrees=80, swing2LimitDegrees=60, twistLimitDegrees=8}
 {bone1Name=jacket_frt_06_r, bone2Name=jacket_frt_05_r, swing1LimitDegrees=80, swing2LimitDegrees=60, twistLimitDegrees=8}
 {bone1Name=jacket_frt_07_r, bone2Name=jacket_frt_06_r, swing1LimitDegrees=80, swing2LimitDegrees=60, twistLimitDegrees=8}
 {bone1Name=jacket_side_01_r, bone2Name=spine_03, swing1LimitDegrees=80, swing2LimitDegrees=60, twistLimitDegrees=8}
 {bone1Name=jacket_side_02_r, bone2Name=jacket_side_01_r, swing1LimitDegrees=80, swing2LimitDegrees=60, twistLimitDegrees=8}
 {bone1Name=jacket_side_03_r, bone2Name=jacket_side_02_r, swing1LimitDegrees=80, swing2LimitDegrees=60, twistLimitDegrees=8}
 {bone1Name=jacket_side_04_r, bone2Name=jacket_side_03_r, swing1LimitDegrees=80, swing2LimitDegrees=60, twistLimitDegrees=8}
 {bone1Name=jacket_side_05_r, bone2Name=jacket_side_04_r, swing1LimitDegrees=80, swing2LimitDegrees=60, twistLimitDegrees=8}
 {bone1Name=jacket_side_06_r, bone2Name=jacket_side_05_r, swing1LimitDegrees=80, swing2LimitDegrees=60, twistLimitDegrees=8}
 {bone1Name=jacket_side_07_r, bone2Name=jacket_side_06_r, swing1LimitDegrees=80, swing2LimitDegrees=60, twistLimitDegrees=8}
 {bone1Name=jacket_side_08_r, bone2Name=jacket_side_07_r, swing1LimitDegrees=80, swing2LimitDegrees=60, twistLimitDegrees=8}
 {bone1Name=jacket_back_01_r, bone2Name=spine_03, swing1LimitDegrees=80, swing2LimitDegrees=60, twistLimitDegrees=8}
 {bone1Name=jacket_back_02_r, bone2Name=jacket_back_01_r, swing1LimitDegrees=80, swing2LimitDegrees=60, twistLimitDegrees=8}
 {bone1Name=jacket_back_03_r, bone2Name=jacket_back_02_r, swing1LimitDegrees=80, swing2LimitDegrees=60, twistLimitDegrees=8}
 {bone1Name=jacket_back_04_r, bone2Name=jacket_back_03_r, swing1LimitDegrees=80, swing2LimitDegrees=60, twistLimitDegrees=8}
 {bone1Name=jacket_back_05_r, bone2Name=jacket_back_04_r, swing1LimitDegrees=80, swing2LimitDegrees=60, twistLimitDegrees=8}
 {bone1Name=jacket_back_02_r, bone2Name=jacket_side_02_r, swing1LimitDegrees=60, swing2LimitDegrees=60, twistLimitDegrees=60}
 {bone1Name=jacket_back_03_r, bone2Name=jacket_side_03_r, swing1LimitDegrees=60, swing2LimitDegrees=60, twistLimitDegrees=60}
 {bone1Name=jacket_back_04_r, bone2Name=jacket_side_04_r, swing1LimitDegrees=60, swing2LimitDegrees=60, twistLimitDegrees=60}
 {bone1Name=jacket_back_03_l, bone2Name=jacket_back_02_l, swing1LimitDegrees=80, swing2LimitDegrees=60, twistLimitDegrees=8}
 {bone1Name=jacket_back_04_l, bone2Name=jacket_back_03_l, swing1LimitDegrees=80, swing2LimitDegrees=60, twistLimitDegrees=8}
 {bone1Name=jacket_back_05_l, bone2Name=jacket_back_04_l, swing1LimitDegrees=80, swing2LimitDegrees=60, twistLimitDegrees=8}
 {bone1Name=jacket_side_04_l, bone2Name=jacket_side_03_l, swing1LimitDegrees=80, swing2LimitDegrees=60, twistLimitDegrees=8}
 {bone1Name=jacket_side_05_l, bone2Name=jacket_side_04_l, swing1LimitDegrees=80, swing2LimitDegrees=60, twistLimitDegrees=8}
 {bone1Name=jacket_side_06_l, bone2Name=jacket_side_05_l, swing1LimitDegrees=80, swing2LimitDegrees=60, twistLimitDegrees=8}
 {bone1Name=jacket_side_07_l, bone2Name=jacket_side_06_l, swing1LimitDegrees=80, swing2LimitDegrees=60, twistLimitDegrees=8}
 {bone1Name=jacket_side_08_l, bone2Name=jacket_side_07_l, swing1LimitDegrees=80, swing2LimitDegrees=60, twistLimitDegrees=8}
 {
  bone1Name=jacket_frt_03_l
  bone2Name=spine_03
  swing1LimitDegrees=80
  swing1Motion=Free
  swing2LimitDegrees=60
  swing2Motion=Free
  twistLimitDegrees=8
  twistMotion=Free
 }
 {
  bone1Name=jacket_frt_04_l
  bone2Name=jacket_frt_03_l
  swing1LimitDegrees=80
  swing1Motion=Free
  swing2LimitDegrees=60
  swing2Motion=Free
  twistLimitDegrees=8
  twistMotion=Free
 }
 {
  bone1Name=jacket_frt_05_l
  bone2Name=jacket_frt_04_l
  swing1LimitDegrees=80
  swing1Motion=Free
  swing2LimitDegrees=60
  swing2Motion=Free
  twistLimitDegrees=8
  twistMotion=Free
 }
 {bone1Name=jacket_frt_06_l, bone2Name=jacket_frt_05_l, swing1LimitDegrees=80, swing2LimitDegrees=60, twistLimitDegrees=8}
 {bone1Name=jacket_frt_07_l, bone2Name=jacket_frt_06_l, swing1LimitDegrees=80, swing2LimitDegrees=60, twistLimitDegrees=8}
 {bone1Name=jacket_frt_08_l, bone2Name=jacket_frt_07_l, swing1LimitDegrees=80, swing2LimitDegrees=60, twistLimitDegrees=8}
 {bone1Name=jacket_frt_09_l, bone2Name=jacket_frt_08_l, swing1LimitDegrees=80, swing2LimitDegrees=60, twistLimitDegrees=8}
 {
  bone1Name=calf_l
  bone2Name=thigh_l
  swing1LimitDegrees=45
  swing1Motion=Free
  swing2LimitDegrees=45
  swing2Motion=Free
  twistLimitDegrees=45
  twistMotion=Free
 }
 {
  bone1Name=foot_l
  bone2Name=calf_l
  swing1LimitDegrees=45
  swing1Motion=Free
  swing2LimitDegrees=45
  swing2Motion=Free
  twistLimitDegrees=45
  twistMotion=Free
 }
 {bone1Name=jacket_back_02, bone2Name=spine_03, swing1LimitDegrees=80, swing2LimitDegrees=60, twistLimitDegrees=8}
 {bone1Name=jacket_back_03, bone2Name=jacket_back_02, swing1LimitDegrees=80, swing2LimitDegrees=60, twistLimitDegrees=8}
 {bone1Name=jacket_back_04, bone2Name=jacket_back_03, swing1LimitDegrees=80, swing2LimitDegrees=60, twistLimitDegrees=8}
 {bone1Name=jacket_back_05, bone2Name=jacket_back_04, swing1LimitDegrees=80, swing2LimitDegrees=60, twistLimitDegrees=8}
 {
  bone1Name=jacket_frtSide_03_l
  bone2Name=spine_03
  swing1LimitDegrees=80
  swing1Motion=Free
  swing2LimitDegrees=60
  swing2Motion=Free
  twistLimitDegrees=8
  twistMotion=Free
 }
 {bone1Name=jacket_frtSide_04_l, bone2Name=jacket_frtSide_03_l, swing1LimitDegrees=80, swing2LimitDegrees=60, twistLimitDegrees=8}
 {bone1Name=jacket_frtSide_05_l, bone2Name=jacket_frtSide_04_l, swing1LimitDegrees=80, swing2LimitDegrees=60, twistLimitDegrees=8}
 {
  bone1Name=jacket_frtSide_03_r
  bone2Name=spine_03
  swing1LimitDegrees=80
  swing1Motion=Free
  swing2LimitDegrees=60
  swing2Motion=Free
  twistLimitDegrees=8
  twistMotion=Free
 }
 {bone1Name=jacket_frtSide_04_r, bone2Name=jacket_frtSide_03_r, swing1LimitDegrees=80, swing2LimitDegrees=60, twistLimitDegrees=8}
 {bone1Name=jacket_frtSide_05_r, bone2Name=jacket_frtSide_04_r, swing1LimitDegrees=80, swing2LimitDegrees=60, twistLimitDegrees=8}
 {bone1Name=jacket_frtSide_04_l, bone2Name=jacket_frt_05_l, swing1LimitDegrees=60, swing2LimitDegrees=60, twistLimitDegrees=60}
 {bone1Name=jacket_frtSide_03_l, bone2Name=jacket_frt_04_l, swing1LimitDegrees=60, swing2LimitDegrees=60, twistLimitDegrees=60}
 {bone1Name=jacket_frtSide_05_l, bone2Name=jacket_frt_06_l, swing1LimitDegrees=60, swing2LimitDegrees=60, twistLimitDegrees=60}
 {bone1Name=jacket_side_03_l, bone2Name=jacket_frtSide_03_l, swing1LimitDegrees=60, swing2LimitDegrees=60, twistLimitDegrees=60}
 {bone1Name=jacket_side_04_l, bone2Name=jacket_frtSide_04_l, swing1LimitDegrees=60, swing2LimitDegrees=60, twistLimitDegrees=60}
 {bone1Name=jacket_back_04_l, bone2Name=jacket_side_04_l, swing1LimitDegrees=60, swing2LimitDegrees=60, twistLimitDegrees=60}
 {bone1Name=jacket_back_02, bone2Name=jacket_back_02_l, swing1LimitDegrees=60, swing2LimitDegrees=60, twistLimitDegrees=60}
 {bone1Name=jacket_back_03, bone2Name=jacket_back_03_l, swing1LimitDegrees=60, swing2LimitDegrees=60, twistLimitDegrees=60}
 {bone1Name=jacket_back_04, bone2Name=jacket_back_04_l, swing1LimitDegrees=60, swing2LimitDegrees=60, twistLimitDegrees=60}
 {bone1Name=jacket_back_05, bone2Name=jacket_back_05_l, swing1LimitDegrees=60, swing2LimitDegrees=60, twistLimitDegrees=60}
 {bone1Name=jacket_frtSide_03_r, bone2Name=jacket_frt_04_r, swing1LimitDegrees=60, swing2LimitDegrees=60, twistLimitDegrees=60}
 {bone1Name=jacket_frtSide_04_r, bone2Name=jacket_frt_05_r, swing1LimitDegrees=60, swing2LimitDegrees=60, twistLimitDegrees=60}
 {bone1Name=jacket_frtSide_05_r, bone2Name=jacket_frt_06_r, swing1LimitDegrees=60, swing2LimitDegrees=60, twistLimitDegrees=60}
 {bone1Name=jacket_back_02, bone2Name=jacket_back_02_r, swing1LimitDegrees=60, swing2LimitDegrees=60, twistLimitDegrees=60}
 {bone1Name=jacket_back_03, bone2Name=jacket_back_03_r, swing1LimitDegrees=60, swing2LimitDegrees=60, twistLimitDegrees=60}
 {bone1Name=jacket_back_04, bone2Name=jacket_back_04_r, swing1LimitDegrees=60, swing2LimitDegrees=60, twistLimitDegrees=60}
 {bone1Name=jacket_back_05, bone2Name=jacket_back_05_r, swing1LimitDegrees=60, swing2LimitDegrees=60, twistLimitDegrees=60}
 {bone1Name=jacket_frtSide_03_r, bone2Name=jacket_side_03_r, swing1LimitDegrees=60, swing2LimitDegrees=60, twistLimitDegrees=60}
 {bone1Name=jacket_frtSide_04_r, bone2Name=jacket_side_04_r, swing1LimitDegrees=60, swing2LimitDegrees=60, twistLimitDegrees=60}
 {bone1Name=jacket_frtSide_05_r, bone2Name=jacket_side_05_r, swing1LimitDegrees=60, swing2LimitDegrees=60, twistLimitDegrees=60}
 {bone1Name=jacket_frtSide_05_l, bone2Name=jacket_side_05_l, swing1LimitDegrees=60, swing2LimitDegrees=60, twistLimitDegrees=60}
 {bone1Name=thigh_l, bone2Name=pelvis, swing1LimitDegrees=45, swing2LimitDegrees=45, twistLimitDegrees=45}
 {bone1Name=jacket_frt_09_r, bone2Name=jacket_frt_08_r, swing1LimitDegrees=80, swing2LimitDegrees=60, twistLimitDegrees=8}
 {bone1Name=jacket_frt_08_r, bone2Name=jacket_frt_07_r, swing1LimitDegrees=80, swing2LimitDegrees=60, twistLimitDegrees=8}
 {
  bone1Name=spine_01
  bone2Name=pelvis
  swing1LimitDegrees=45
  swing1Motion=Free
  swing2LimitDegrees=45
  swing2Motion=Free
  twistLimitDegrees=45
  twistMotion=Free
 }
 {bone1Name=jacket_side_02_l, bone2Name=jacket_side_01_l, swing1LimitDegrees=80, swing2LimitDegrees=60, twistLimitDegrees=8}
 {bone1Name=jacket_side_01_l, bone2Name=spine_03, swing1LimitDegrees=80, swing2LimitDegrees=60, twistLimitDegrees=8}
 {bone1Name=jacket_side_03_l, bone2Name=jacket_side_02_l, swing1LimitDegrees=80, swing2LimitDegrees=60, twistLimitDegrees=8}
 {bone1Name=jacket_back_02_l, bone2Name=jacket_back_01_l, swing1LimitDegrees=80, swing2LimitDegrees=60, twistLimitDegrees=8}
 {bone1Name=jacket_back_01_l, bone2Name=spine_03, swing1LimitDegrees=80, swing2LimitDegrees=60, twistLimitDegrees=8}
 {bone1Name=thigh_r, bone2Name=pelvis, swing1LimitDegrees=80, swing2LimitDegrees=60, twistLimitDegrees=8}
]
