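# Physics asset for the female medium Painter body (Sky) that models a long jacket over the lower body and torso.
# Capsules cover pelvis, spine_03, thighs, calves and feet, while box chains hang from jacket_base as front, front-side, side and back panels.
# Front and side chains end in small, tightly limited capsules.
# Cross constraints link neighboring chains into a cloth-like lattice; some upper links, mostly left and front-side, have free swing and twist.
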PhysicsAsset /BRCosmetics/Characters/Player/Female/Medium/Bodies/F_MED_Painter/Meshes/F_MED_Painter_Phys_Sky
bodies {
 calf_l [{shapeType=Capsule, length=49.722, radius=7.5, center=<-23.2226 -4.6248 0>, rotation=<rot 90 0 175>}]
 calf_r [
  {shapeType=Capsule, length=5.472, radius=6.75, center=<9.7765 3.4485 0.52>, rotation=<rot 75.4777 -174.9999 -180>}
  {shapeType=Capsule, length=15.472, radius=8.5, center=<3.383 2.5887 -0.2914>, rotation=<rot 82.8936 -174.999 176.1209>}
  {shapeType=Capsule, length=21.222, radius=6, center=<30.2472 3.4652 1.1432>, rotation=<rot 88.1308 -104.365 -98.4575>}
  {shapeType=Capsule, length=4.472, radius=6.25, center=<16.4372 4.3168 0.7899>, rotation=<rot 88.0231 -174.998 -174.9757>}
 ]
 foot_l [{shapeType=Capsule, length=12.6052, radius=6.5119, center=<-2.9081 7.2701 -0.3337>, rotation=<rot 0 5.1452 90.9165>}]
 foot_r [{shapeType=Capsule, length=12.6052, radius=6.5119, center=<2.9081 -7.2701 0.3337>, rotation=<rot 0 5.1452 -89.0835>}]
 jacket_back_01 [{shapeType=Box, extentX=8.01, extentY=1.51, extentZ=8.51}]
 jacket_back_01_l [{shapeType=Box, extentX=8.01, extentY=1.51, extentZ=8.51, rotation=<rot 0.2457 0.0003 -0.0001>}]
 jacket_back_01_r [{shapeType=Box, extentX=8.01, extentY=1.51, extentZ=8.51, rotation=<rot -3.1744 0.0003 0>}]
 jacket_back_02 [{shapeType=Box, extentX=16.1795, extentY=2.1524, extentZ=17.4777, center=<-4.7346 0.1686 0.1294>}]
 jacket_back_02_l [{shapeType=Box, extentX=16.1795, extentY=2.1524, extentZ=15.4777, center=<-4.7346 0.1686 0.1294>, rotation=<rot -0.0302 -1.6396 -4.5639>}]
 jacket_back_02_r [{shapeType=Box, extentX=16.1795, extentY=2.1524, extentZ=15.4777, center=<-4.7346 0.1686 0.1294>, rotation=<rot -1.3333 -0.9236 6.0094>}]
 jacket_back_03 [{shapeType=Box, extentX=14.1767, extentY=3.528, extentZ=17.1549, center=<-6.2146 0.2753 0.1272>}]
 jacket_back_03_l [{shapeType=Box, extentX=14.1767, extentY=3.028, extentZ=17.1549, center=<-6.2146 0.2753 0.1272>, rotation=<rot 1.1527 0.3833 -1.3369>}]
 jacket_back_03_r [{shapeType=Box, extentX=14.1767, extentY=3.528, extentZ=15.1549, center=<-6.2146 0.2753 0.1272>, rotation=<rot -2.2831 0.7189 6.6743>}]
 jacket_back_04 [{shapeType=Box, extentX=14.4493, extentY=2.7883, extentZ=15.7545, center=<-7.1513 -0.2294 0.0135>}]
 jacket_back_04_l [{shapeType=Box, extentX=14.4493, extentY=3.2883, extentZ=16.7545, center=<-7.1513 -0.2294 0.0135>, rotation=<rot 0.7153 3.5283 -6.7926>}]
 jacket_back_04_r [{shapeType=Box, extentX=14.4493, extentY=2.7883, extentZ=15.7545, center=<-7.1513 -0.2294 0.0135>, rotation=<rot -3.245 2.3728 7.3244>}]
 jacket_back_05 [{shapeType=Box, extentX=10.572, extentY=2.0221, extentZ=16.3509, center=<-5.393 -0.7495 0.0701>}]
 jacket_back_05_l [{shapeType=Box, extentX=10.572, extentY=1.5221, extentZ=16.3509, center=<-5.1503 -0.7248 0.0471>, rotation=<rot 1.7164 -0.383 -6.336>}]
 jacket_back_05_r [{shapeType=Box, extentX=10.572, extentY=2.0221, extentZ=16.3509, center=<-5.393 -0.7495 0.0701>, rotation=<rot -2.7506 1.8068 9.1855>}]
 jacket_base [{shapeType=Capsule, length=0.505, radius=0.505, rotation=<rot 0 -90 0>}]
 jacket_frtSide_01_l [{shapeType=Box, extentX=8.01, extentY=1.51, extentZ=8.51}]
 jacket_frtSide_01_r [{shapeType=Box, extentX=8.01, extentY=1.51, extentZ=8.51}]
 jacket_frtSide_02_l [{shapeType=Box, extentX=16.1795, extentY=2.6524, extentZ=14.9777, center=<-4.7346 0.1686 0.1294>}]
 jacket_frtSide_02_r [{shapeType=Box, extentX=16.1795, extentY=2.1524, extentZ=14.9777, center=<-4.7346 0.1686 0.1294>}]
 jacket_frtSide_03_l [{shapeType=Box, extentX=18.6767, extentY=2.528, extentZ=15.6549, center=<-6.2146 0.2753 0.1272>}]
 jacket_frtSide_03_r [{shapeType=Box, extentX=18.6767, extentY=2.528, extentZ=15.6549, center=<-6.2146 0.2753 0.1272>}]
 jacket_frtSide_04_l [{shapeType=Box, extentX=14.4493, extentY=3.2883, extentZ=16.2545, center=<-7.1513 -0.2294 0.0135>}]
 jacket_frtSide_04_r [{shapeType=Box, extentX=14.4493, extentY=2.7883, extentZ=16.2545, center=<-7.1513 -0.2294 0.0135>}]
 jacket_frtSide_05_l [{shapeType=Box, extentX=10.572, extentY=2.0221, extentZ=16.3509, center=<-4.763 -0.7273 0.0587>}]
 jacket_frtSide_05_r [{shapeType=Box, extentX=10.572, extentY=2.0221, extentZ=16.3509, center=<-4.763 -0.7273 0.0587>}]
 jacket_frt_01_l [{shapeType=Box, extentX=1.01, extentY=7.01, extentZ=5.01, rotation=<rot 0 -90 0>}]
 jacket_frt_01_r [{shapeType=Box, extentX=1.01, extentY=7.01, extentZ=5.01, rotation=<rot 0 -90 0>}]
 jacket_frt_02_l [{shapeType=Box, extentX=10.51, extentY=1.51, extentZ=8.01, center=<0 0.0001 0.4213>, rotation=<rot 0 0.0002 0>}]
 jacket_frt_02_r [{shapeType=Box, extentX=10.51, extentY=1.01, extentZ=8.01, center=<0 0.0001 0.4213>, rotation=<rot 0 0.0002 0>}]
 jacket_frt_03_l [{shapeType=Box, extentX=17.9595, extentY=2.4164, extentZ=12.0088, center=<-5.2742 0.092 0.2215>, rotation=<rot 1.2626 2.7849 0.0395>}]
 jacket_frt_03_r [{shapeType=Box, extentX=17.9595, extentY=2.9164, extentZ=12.0088, center=<-5.2637 0.5111 1.4867>, rotation=<rot 1.2633 0.998 0.0001>}]
 jacket_frt_04_l [{shapeType=Box, extentX=20.0893, extentY=2.4003, extentZ=11.306, center=<-7.0475 0.1723 0.1425>, rotation=<rot -1.7478 1.1478 -0.0586>}]
 jacket_frt_04_r [{shapeType=Box, extentX=20.0893, extentY=4.4003, extentZ=11.306, center=<-7.0613 0.8649 0.1433>, rotation=<rot 0.4059 1.3897 -0.066>}]
 jacket_frt_05_l [{shapeType=Box, extentX=14.5625, extentY=2.6218, extentZ=11.6157, center=<-6.7324 1.0278 -0.0582>, rotation=<rot 0.1865 2.0904 0.0065>}]
 jacket_frt_05_r [{shapeType=Box, extentX=14.5625, extentY=3.1218, extentZ=11.6157, center=<-6.7324 1.0278 -0.0582>, rotation=<rot 0.1864 2.5247 0.0079>}]
 jacket_frt_06_l [{shapeType=Box, extentX=13.4127, extentY=2.5628, extentZ=11.1086, center=<-3.7597 0.4594 0.2202>, rotation=<rot -0.5596 0.9685 -0.0289>}]
 jacket_frt_06_r [{shapeType=Box, extentX=13.4127, extentY=2.5628, extentZ=11.1086, center=<-3.7597 0.4594 0.2202>, rotation=<rot -0.5596 0.9685 -0.0289>}]
 jacket_frt_07_l [{shapeType=Capsule, length=5.266, radius=2.2038, shapeName=jacket_frt_07_l_capsule, center=<-2 0 0>, rotation=<rot -90 0 0>}]
 jacket_frt_07_r [{shapeType=Capsule, length=6.2769, radius=1.8147, shapeName=jacket_frt_07_r_capsule, center=<-2 0 0>, rotation=<rot -90 0 0>}]
 jacket_frt_08_l [{shapeType=Capsule, length=3, radius=2, shapeName=jacket_frt_08_l_capsule, center=<-2 0 0>, rotation=<rot -90 0 0>}]
 jacket_frt_08_r [{shapeType=Capsule, length=3, radius=2, shapeName=jacket_frt_08_r_capsule, center=<-2 0 0>, rotation=<rot -90 0 0>}]
 jacket_frt_09_l [{shapeType=Capsule, length=3.9607, radius=1.2141, shapeName=jacket_frt_09_l_capsule, center=<-2 0 0>, rotation=<rot -90 0 0>}]
 jacket_frt_09_r [{shapeType=Capsule, length=3.9592, radius=1.2165, shapeName=jacket_frt_09_r_capsule, center=<-2 0 0>, rotation=<rot -90 0 0>}]
 jacket_side_01_l [{shapeType=Box, extentX=8.01, extentY=1.51, extentZ=8.51, rotation=<rot 1.8986 0.0003 0>}]
 jacket_side_01_r [{shapeType=Box, extentX=8.01, extentY=1.51, extentZ=8.51, rotation=<rot -3.1744 0.0003 0>}]
 jacket_side_02_l [{shapeType=Box, extentX=16, extentY=2.5, extentZ=16, center=<-4.7346 0.1686 0.1294>, rotation=<rot -1.0549 2.8227 7.9854>}]
 jacket_side_02_r [{shapeType=Box, extentX=16, extentY=2.5, extentZ=15, center=<-4.7346 0.1686 0.1294>, rotation=<rot 0.8058 1.1507 -12.1309>}]
 jacket_side_03_l [{shapeType=Box, extentX=19, extentY=3, extentZ=17, center=<-6.2146 0.2753 0.1272>, rotation=<rot -1.3204 1.1242 9.5219>}]
 jacket_side_03_r [{shapeType=Box, extentX=19, extentY=2, extentZ=15, center=<-6.2146 0.2753 0.1272>, rotation=<rot 1.8605 -0.4969 -7.4838>}]
 jacket_side_04_l [{shapeType=Box, extentX=14, extentY=2.5, extentZ=16.5, center=<-7.1513 -0.2294 0.0135>, rotation=<rot -1.6056 2.2678 9.9341>}]
 jacket_side_04_r [{shapeType=Box, extentX=14, extentY=2, extentZ=15, center=<-7.1513 -0.2294 0.0135>, rotation=<rot 0.6989 3.194 -10.8824>}]
 jacket_side_05_l [{shapeType=Box, extentX=10, extentY=2.5, extentZ=15, center=<-4.763 -0.7273 0.0587>, rotation=<rot -0.7513 -0.2819 7.7776>}]
 jacket_side_05_r [{shapeType=Box, extentX=10, extentY=2, extentZ=15, center=<-4.763 -0.7273 0.0587>, rotation=<rot -0.7513 -0.2819 -9.5575>}]
 jacket_side_06_l [{shapeType=Capsule, length=3.9094, radius=2.4968, shapeName=jacket_side_06_l_capsule, center=<-2 0 0>, rotation=<rot -90 0 0>}]
 jacket_side_06_r [{shapeType=Capsule, length=3.9435, radius=2.5321, shapeName=jacket_side_06_r_capsule, center=<-2 0 0>, rotation=<rot -90 0 0>}]
 jacket_side_07_l [{shapeType=Capsule, length=3, radius=2, shapeName=jacket_side_07_l_capsule, center=<-2 0 0>, rotation=<rot -90 0 0>}]
 jacket_side_07_r [{shapeType=Capsule, length=3, radius=2, shapeName=jacket_side_07_r_capsule, center=<-2 0 0>, rotation=<rot -90 0 0>}]
 jacket_side_08_l [{shapeType=Capsule, length=3.7928, radius=1.5295, shapeName=jacket_side_08_l_capsule, center=<-2 0 0>, rotation=<rot -90 0 0>}]
 jacket_side_08_r [{shapeType=Capsule, length=3.8316, radius=1.5208, shapeName=jacket_side_08_r_capsule, center=<-2 0 0>, rotation=<rot -90 0 0>}]
 pelvis [
  {shapeType=Capsule, length=4.5, radius=11.5, center=<-4.3449 -1.1221 -0.0304>}
  {shapeType=Capsule, length=5, radius=11, center=<-4.3448 -1.06 1.1187>, rotation=<rot 0 0 -4.2501>}
  {shapeType=Capsule, length=5, radius=11, center=<-4.3447 -1.8045 -1.402>, rotation=<rot 0 0 2.2241>}
 ]
 spine_03 [{shapeType=Capsule, length=14.8737, radius=11.1435, center=<-4.5611 1.7894 -0.1524>, rotation=<rot 6.275 -83.303 90.0452>}]
 thigh_l [
  {shapeType=Capsule, length=8, radius=7.5, center=<-20.9471 -0.0908 0.161>, rotation=<rot 87.6152 -12.0017 165.9154>}
  {shapeType=Capsule, length=5.5, radius=6.5, center=<-31.9914 -1.0011 0.2198>, rotation=<rot -84.9731 -176.8868 2.0178>}
  {shapeType=Capsule, length=12.75, radius=10.5, center=<-8.9424 -0.9155 1.7293>, rotation=<rot 84.7183 170.4672 -1.4949>}
 ]
 thigh_r [
  {shapeType=Capsule, length=6.75, radius=7.75, center=<21.659 0.0426 -0.4051>, rotation=<rot 88.1356 -32.1733 -33.8667>}
  {shapeType=Capsule, length=6.25, radius=6, center=<33.0085 1.0916 -0.0192>, rotation=<rot -88.0216 179.9998 -174.8738>}
  {shapeType=Capsule, length=10.5, radius=9.75, center=<8.8453 1.2264 -1.0746>, rotation=<rot 84.2916 -170.1611 -160.9737>}
 ]
}
constraints [
 {bone1Name=spine_03, bone2Name=pelvis, swing1LimitDegrees=45, swing2LimitDegrees=45, twistLimitDegrees=45}
 {bone1Name=thigh_r, bone2Name=pelvis, swing1LimitDegrees=45, swing2LimitDegrees=45, twistLimitDegrees=45}
 {bone1Name=calf_r, bone2Name=thigh_r, swing1LimitDegrees=45, swing2LimitDegrees=45, twistLimitDegrees=45}
 {
  bone1Name=foot_r
  bone2Name=calf_r
  swing1LimitDegrees=45
  swing1Motion=Free
  swing2LimitDegrees=45
  swing2Motion=Free
  twistLimitDegrees=45
  twistMotion=Free
 }
 {
  bone1Name=jacket_base
  bone2Name=spine_03
  swing1LimitDegrees=45
  swing1Motion=Free
  swing2LimitDegrees=45
  swing2Motion=Free
  twistLimitDegrees=45
  twistMotion=Free
 }
 {bone1Name=jacket_frt_01_r, bone2Name=jacket_base, swing1LimitDegrees=80, swing2LimitDegrees=60, twistLimitDegrees=8}
 {bone1Name=jacket_frt_02_r, bone2Name=jacket_frt_01_r, swing1LimitDegrees=80, swing2LimitDegrees=60, twistLimitDegrees=8}
 {bone1Name=jacket_frt_03_r, bone2Name=jacket_frt_02_r, swing1LimitDegrees=80, swing2LimitDegrees=60, twistLimitDegrees=8}
 {bone1Name=jacket_frt_04_r, bone2Name=jacket_frt_03_r, swing1LimitDegrees=80, swing2LimitDegrees=60, twistLimitDegrees=8}
 {bone1Name=jacket_frt_05_r, bone2Name=jacket_frt_04_r, swing1LimitDegrees=80, swing2LimitDegrees=60, twistLimitDegrees=8}
 {bone1Name=jacket_frt_06_r, bone2Name=jacket_frt_05_r, swing1LimitDegrees=80, swing2LimitDegrees=60, twistLimitDegrees=8}
 {bone1Name=jacket_frt_07_r, bone2Name=jacket_frt_06_r, swing1LimitDegrees=40, swing2LimitDegrees=20, twistLimitDegrees=2}
 {bone1Name=jacket_frt_08_r, bone2Name=jacket_frt_07_r, swing1LimitDegrees=40, swing2LimitDegrees=20, twistLimitDegrees=2}
 {bone1Name=jacket_frt_09_r, bone2Name=jacket_frt_08_r, swing1LimitDegrees=40, swing2LimitDegrees=20, twistLimitDegrees=2}
 {bone1Name=jacket_side_01_r, bone2Name=jacket_base, swing1LimitDegrees=80, swing2LimitDegrees=60, twistLimitDegrees=8}
 {bone1Name=jacket_side_02_r, bone2Name=jacket_side_01_r, swing1LimitDegrees=80, swing2LimitDegrees=60, twistLimitDegrees=8}
 {bone1Name=jacket_side_03_r, bone2Name=jacket_side_02_r, swing1LimitDegrees=80, swing2LimitDegrees=60, twistLimitDegrees=8}
 {bone1Name=jacket_side_04_r, bone2Name=jacket_side_03_r, swing1LimitDegrees=80, swing2LimitDegrees=60, twistLimitDegrees=8}
 {bone1Name=jacket_side_05_r, bone2Name=jacket_side_04_r, swing1LimitDegrees=80, swing2LimitDegrees=60, twistLimitDegrees=8}
 {bone1Name=jacket_side_06_r, bone2Name=jacket_side_05_r, swing1LimitDegrees=40, swing2LimitDegrees=20, twistLimitDegrees=1}
 {bone1Name=jacket_side_07_r, bone2Name=jacket_side_06_r, swing1LimitDegrees=40, swing2LimitDegrees=20, twistLimitDegrees=1}
 {bone1Name=jacket_side_08_r, bone2Name=jacket_side_07_r, swing1LimitDegrees=40, swing2LimitDegrees=20, twistLimitDegrees=1}
 {bone1Name=jacket_back_01_r, bone2Name=jacket_base, swing1LimitDegrees=80, swing2LimitDegrees=60, twistLimitDegrees=8}
 {bone1Name=jacket_back_02_r, bone2Name=jacket_back_01_r, swing1LimitDegrees=80, swing2LimitDegrees=60, twistLimitDegrees=8}
 {bone1Name=jacket_back_03_r, bone2Name=jacket_back_02_r, swing1LimitDegrees=80, swing2LimitDegrees=60, twistLimitDegrees=8}
 {bone1Name=jacket_back_04_r, bone2Name=jacket_back_03_r, swing1LimitDegrees=80, swing2LimitDegrees=60, twistLimitDegrees=8}
 {bone1Name=jacket_back_05_r, bone2Name=jacket_back_04_r, swing1LimitDegrees=80, swing2LimitDegrees=60, twistLimitDegrees=8}
 {bone1Name=jacket_back_02_r, bone2Name=jacket_side_02_r, swing1LimitDegrees=60, swing2LimitDegrees=60, twistLimitDegrees=60}
 {bone1Name=jacket_back_03_r, bone2Name=jacket_side_03_r, swing1LimitDegrees=60, swing2LimitDegrees=60, twistLimitDegrees=60}
 {bone1Name=jacket_back_04_r, bone2Name=jacket_side_04_r, swing1LimitDegrees=60, swing2LimitDegrees=60, twistLimitDegrees=60}
 {bone1Name=jacket_back_01_l, bone2Name=jacket_base, swing1LimitDegrees=80, swing2LimitDegrees=60, twistLimitDegrees=8}
 {bone1Name=jacket_back_02_l, bone2Name=jacket_back_01_l, swing1LimitDegrees=80, swing2LimitDegrees=60, twistLimitDegrees=8}
 {bone1Name=jacket_back_03_l, bone2Name=jacket_back_02_l, swing1LimitDegrees=80, swing2LimitDegrees=60, twistLimitDegrees=8}
 {bone1Name=jacket_back_04_l, bone2Name=jacket_back_03_l, swing1LimitDegrees=80, swing2LimitDegrees=60, twistLimitDegrees=8}
 {bone1Name=jacket_back_05_l, bone2Name=jacket_back_04_l, swing1LimitDegrees=80, swing2LimitDegrees=60, twistLimitDegrees=8}
 {
  bone1Name=jacket_side_01_l
  bone2Name=jacket_base
  swing1LimitDegrees=80
  swing1Motion=Free
  swing2LimitDegrees=60
  swing2Motion=Free
  twistLimitDegrees=8
  twistMotion=Free
 }
 {
  bone1Name=jacket_side_02_l
  bone2Name=jacket_side_01_l
  swing1LimitDegrees=80
  swing1Motion=Free
  swing2LimitDegrees=60
  swing2Motion=Free
  twistLimitDegrees=8
  twistMotion=Free
 }
 {
  bone1Name=jacket_side_03_l
  bone2Name=jacket_side_02_l
  swing1LimitDegrees=80
  swing1Motion=Free
  swing2LimitDegrees=60
  swing2Motion=Free
  twistLimitDegrees=8
  twistMotion=Free
 }
 {bone1Name=jacket_side_04_l, bone2Name=jacket_side_03_l, swing1LimitDegrees=80, swing2LimitDegrees=60, twistLimitDegrees=8}
 {bone1Name=jacket_side_05_l, bone2Name=jacket_side_04_l, swing1LimitDegrees=80, swing2LimitDegrees=60, twistLimitDegrees=8}
 {bone1Name=jacket_side_06_l, bone2Name=jacket_side_05_l, swing1LimitDegrees=40, swing2LimitDegrees=20, twistLimitDegrees=1}
 {bone1Name=jacket_side_07_l, bone2Name=jacket_side_06_l, swing1LimitDegrees=40, swing2LimitDegrees=20, twistLimitDegrees=1}
 {bone1Name=jacket_side_08_l, bone2Name=jacket_side_07_l, swing1LimitDegrees=40, swing2LimitDegrees=20, twistLimitDegrees=1}
 {
  bone1Name=jacket_frt_01_l
  bone2Name=jacket_base
  swing1LimitDegrees=80
  swing1Motion=Free
  swing2LimitDegrees=60
  swing2Motion=Free
  twistLimitDegrees=8
  twistMotion=Free
 }
 {
  bone1Name=jacket_frt_02_l
  bone2Name=jacket_frt_01_l
  swing1LimitDegrees=80
  swing1Motion=Free
  swing2LimitDegrees=60
  swing2Motion=Free
  twistLimitDegrees=8
  twistMotion=Free
 }
 {
  bone1Name=jacket_frt_03_l
  bone2Name=jacket_frt_02_l
  swing1LimitDegrees=80
  swing1Motion=Free
  swing2LimitDegrees=60
  swing2Motion=Free
  twistLimitDegrees=8
  twistMotion=Free
 }
 {
  bone1Name=jacket_frt_04_l
  bone2Name=jacket_frt_03_l
  swing1LimitDegrees=80
  swing1Motion=Free
  swing2LimitDegrees=60
  swing2Motion=Free
  twistLimitDegrees=8
  twistMotion=Free
 }
 {
  bone1Name=jacket_frt_05_l
  bone2Name=jacket_frt_04_l
  swing1LimitDegrees=80
  swing1Motion=Free
  swing2LimitDegrees=60
  swing2Motion=Free
  twistLimitDegrees=8
  twistMotion=Free
 }
 {bone1Name=jacket_frt_06_l, bone2Name=jacket_frt_05_l, swing1LimitDegrees=80, swing2LimitDegrees=60, twistLimitDegrees=8}
 {bone1Name=jacket_frt_07_l, bone2Name=jacket_frt_06_l, swing1LimitDegrees=40, swing2LimitDegrees=20, twistLimitDegrees=2}
 {bone1Name=jacket_frt_08_l, bone2Name=jacket_frt_07_l, swing1LimitDegrees=40, swing2LimitDegrees=20, twistLimitDegrees=2}
 {bone1Name=jacket_frt_09_l, bone2Name=jacket_frt_08_l, swing1LimitDegrees=40, swing2LimitDegrees=20, twistLimitDegrees=2}
 {
  bone1Name=calf_l
  bone2Name=thigh_l
  swing1LimitDegrees=45
  swing1Motion=Free
  swing2LimitDegrees=45
  swing2Motion=Free
  twistLimitDegrees=45
  twistMotion=Free
 }
 {
  bone1Name=foot_l
  bone2Name=calf_l
  swing1LimitDegrees=45
  swing1Motion=Free
  swing2LimitDegrees=45
  swing2Motion=Free
  twistLimitDegrees=45
  twistMotion=Free
 }
 {bone1Name=jacket_back_01, bone2Name=jacket_base, swing1LimitDegrees=80, swing2LimitDegrees=60, twistLimitDegrees=8}
 {bone1Name=jacket_back_02, bone2Name=jacket_back_01, swing1LimitDegrees=80, swing2LimitDegrees=60, twistLimitDegrees=8}
 {bone1Name=jacket_back_03, bone2Name=jacket_back_02, swing1LimitDegrees=80, swing2LimitDegrees=60, twistLimitDegrees=8}
 {bone1Name=jacket_back_04, bone2Name=jacket_back_03, swing1LimitDegrees=80, swing2LimitDegrees=60, twistLimitDegrees=8}
 {bone1Name=jacket_back_05, bone2Name=jacket_back_04, swing1LimitDegrees=80, swing2LimitDegrees=60, twistLimitDegrees=8}
 {
  bone1Name=jacket_frtSide_01_l
  bone2Name=jacket_base
  swing1LimitDegrees=80
  swing1Motion=Free
  swing2LimitDegrees=60
  swing2Motion=Free
  twistLimitDegrees=8
  twistMotion=Free
 }
 {
  bone1Name=jacket_frtSide_02_l
  bone2Name=jacket_frtSide_01_l
  swing1LimitDegrees=80
  swing1Motion=Free
  swing2LimitDegrees=60
  swing2Motion=Free
  twistLimitDegrees=8
  twistMotion=Free
 }
 {
  bone1Name=jacket_frtSide_03_l
  bone2Name=jacket_frtSide_02_l
  swing1LimitDegrees=80
  swing1Motion=Free
  swing2LimitDegrees=60
  swing2Motion=Free
  twistLimitDegrees=8
  twistMotion=Free
 }
 {bone1Name=jacket_frtSide_04_l, bone2Name=jacket_frtSide_03_l, swing1LimitDegrees=80, swing2LimitDegrees=60, twistLimitDegrees=8}
 {bone1Name=jacket_frtSide_05_l, bone2Name=jacket_frtSide_04_l, swing1LimitDegrees=80, swing2LimitDegrees=60, twistLimitDegrees=8}
 {
  bone1Name=jacket_frtSide_01_r
  bone2Name=jacket_base
  swing1LimitDegrees=80
  swing1Motion=Free
  swing2LimitDegrees=60
  swing2Motion=Free
  twistLimitDegrees=8
  twistMotion=Free
 }
 {
  bone1Name=jacket_frtSide_02_r
  bone2Name=jacket_frtSide_01_r
  swing1LimitDegrees=80
  swing1Motion=Free
  swing2LimitDegrees=60
  swing2Motion=Free
  twistLimitDegrees=8
  twistMotion=Free
 }
 {
  bone1Name=jacket_frtSide_03_r
  bone2Name=jacket_frtSide_02_r
  swing1LimitDegrees=80
  swing1Motion=Free
  swing2LimitDegrees=60
  swing2Motion=Free
  twistLimitDegrees=8
  twistMotion=Free
 }
 {bone1Name=jacket_frtSide_04_r, bone2Name=jacket_frtSide_03_r, swing1LimitDegrees=80, swing2LimitDegrees=60, twistLimitDegrees=8}
 {bone1Name=jacket_frtSide_05_r, bone2Name=jacket_frtSide_04_r, swing1LimitDegrees=80, swing2LimitDegrees=60, twistLimitDegrees=8}
 {bone1Name=jacket_frtSide_02_l, bone2Name=jacket_frt_03_l, swing1LimitDegrees=60, swing2LimitDegrees=60, twistLimitDegrees=60}
 {bone1Name=jacket_frtSide_04_l, bone2Name=jacket_frt_05_l, swing1LimitDegrees=60, swing2LimitDegrees=60, twistLimitDegrees=60}
 {bone1Name=jacket_frtSide_03_l, bone2Name=jacket_frt_04_l, swing1LimitDegrees=60, swing2LimitDegrees=60, twistLimitDegrees=60}
 {bone1Name=jacket_frtSide_05_l, bone2Name=jacket_frt_06_l, swing1LimitDegrees=60, swing2LimitDegrees=60, twistLimitDegrees=60}
 {bone1Name=jacket_side_02_l, bone2Name=jacket_frtSide_02_l, swing1LimitDegrees=60, swing2LimitDegrees=60, twistLimitDegrees=60}
 {bone1Name=jacket_side_03_l, bone2Name=jacket_frtSide_03_l, swing1LimitDegrees=60, swing2LimitDegrees=60, twistLimitDegrees=60}
 {bone1Name=jacket_side_04_l, bone2Name=jacket_frtSide_04_l, swing1LimitDegrees=60, swing2LimitDegrees=60, twistLimitDegrees=60}
 {bone1Name=jacket_back_02_l, bone2Name=jacket_side_02_l, swing1LimitDegrees=60, swing2LimitDegrees=60, twistLimitDegrees=60}
 {bone1Name=jacket_back_03_l, bone2Name=jacket_side_03_l, swing1LimitDegrees=60, swing2LimitDegrees=60, twistLimitDegrees=60}
 {bone1Name=jacket_back_04_l, bone2Name=jacket_side_04_l, swing1LimitDegrees=60, swing2LimitDegrees=60, twistLimitDegrees=60}
 {bone1Name=jacket_back_02, bone2Name=jacket_back_02_l, swing1LimitDegrees=60, swing2LimitDegrees=60, twistLimitDegrees=60}
 {bone1Name=jacket_back_03, bone2Name=jacket_back_03_l, swing1LimitDegrees=60, swing2LimitDegrees=60, twistLimitDegrees=60}
 {bone1Name=jacket_back_04, bone2Name=jacket_back_04_l, swing1LimitDegrees=60, swing2LimitDegrees=60, twistLimitDegrees=60}
 {bone1Name=jacket_back_05, bone2Name=jacket_back_05_l, swing1LimitDegrees=60, swing2LimitDegrees=60, twistLimitDegrees=60}
 {bone1Name=jacket_frtSide_02_r, bone2Name=jacket_frt_03_r, swing1LimitDegrees=60, swing2LimitDegrees=60, twistLimitDegrees=60}
 {bone1Name=jacket_frtSide_03_r, bone2Name=jacket_frt_04_r, swing1LimitDegrees=60, swing2LimitDegrees=60, twistLimitDegrees=60}
 {bone1Name=jacket_frtSide_04_r, bone2Name=jacket_frt_05_r, swing1LimitDegrees=60, swing2LimitDegrees=60, twistLimitDegrees=60}
 {bone1Name=jacket_frtSide_05_r, bone2Name=jacket_frt_06_r, swing1LimitDegrees=60, swing2LimitDegrees=60, twistLimitDegrees=60}
 {bone1Name=jacket_back_02, bone2Name=jacket_back_02_r, swing1LimitDegrees=60, swing2LimitDegrees=60, twistLimitDegrees=60}
 {bone1Name=jacket_back_03, bone2Name=jacket_back_03_r, swing1LimitDegrees=60, swing2LimitDegrees=60, twistLimitDegrees=60}
 {bone1Name=jacket_back_04, bone2Name=jacket_back_04_r, swing1LimitDegrees=60, swing2LimitDegrees=60, twistLimitDegrees=60}
 {bone1Name=jacket_back_05, bone2Name=jacket_back_05_r, swing1LimitDegrees=60, swing2LimitDegrees=60, twistLimitDegrees=60}
 {bone1Name=jacket_frtSide_02_r, bone2Name=jacket_side_02_r, swing1LimitDegrees=60, swing2LimitDegrees=60, twistLimitDegrees=60}
 {bone1Name=jacket_frtSide_03_r, bone2Name=jacket_side_03_r, swing1LimitDegrees=60, swing2LimitDegrees=60, twistLimitDegrees=60}
 {bone1Name=jacket_frtSide_04_r, bone2Name=jacket_side_04_r, swing1LimitDegrees=60, swing2LimitDegrees=60, twistLimitDegrees=60}
 {bone1Name=jacket_frtSide_05_r, bone2Name=jacket_side_05_r, swing1LimitDegrees=60, swing2LimitDegrees=60, twistLimitDegrees=60}
 {bone1Name=jacket_frtSide_05_l, bone2Name=jacket_side_05_l, swing1LimitDegrees=60, swing2LimitDegrees=60, twistLimitDegrees=60}
 {bone1Name=thigh_l, bone2Name=pelvis, swing1LimitDegrees=45, swing2LimitDegrees=45, twistLimitDegrees=45}
 {bone1Name=jacket_back_05_r, bone2Name=jacket_side_05_r, swing1LimitDegrees=60, swing2LimitDegrees=60, twistLimitDegrees=60}
]
