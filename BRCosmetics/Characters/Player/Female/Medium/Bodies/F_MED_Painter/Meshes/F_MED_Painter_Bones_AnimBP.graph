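# Post-process-style AnimBP for the F_MED_Painter character body, adding secondary motion to hair, hood and jacket over the pose copied from its parent mesh.
# AnimDynamics hair and spine-driven hood/jacket offsets feed a cached pose.
# Front-end or in-game RigidBody simulation follows, plus jacket offsets for skydiving and crouching with long guns.
# Weighted bone constraints replace that branch when rigid bodies are disabled or LOD exceeds 1; skydive/parachute events reset dynamics.
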AnimBlueprint /BRCosmetics/Characters/Player/Female/Medium/Bodies/F_MED_Painter/Meshes/F_MED_Painter_Bones_AnimBP
target_skeleton=&/F_MED_Painter_Bones_Skeleton
graphs [
 AnimGraph {nodes [
   n1 = RigidBody(ComponentPose=Local To Component(LocalPose=Use cached pose 'AnimCache'())) {
    bClampLinearTranslationLimitToRefPose=true
    bOverrideWorldGravity=true
    overridePhysicsAsset=&/F_MED_Painter_Phys_Sky
    simulationSpace=BaseBoneSpace
    alphaBoolBlend {blendOption=Linear}
    baseBoneRef {boneName=spine_04}
    componentAppliedLinearAccClamp=<900 600 600>
    componentLinearAccScale=<0.1 0.1 0.1>
    componentLinearVelScale=<0.1 0.1 0.1>
    overrideWorldGravity {z=-150}
   }
   Blend Poses by bool(BlendPose_0=Component To Local(ComponentPose=@n1), BlendTime_0=0.3, BlendTime_1=0.5, bActiveValue=Get bIsSkyDiving()) {blendType=Linear}
   Component To Local()
   n2 = Get Sim Space Settings() -> (Sim Space Settings:Sim Space Settings Structure)
   n3 = RigidBody!(
    ComponentPose=Local To Component(LocalPose=Use cached pose 'AnimCache'())
    Alpha
    OverrideWorldGravity:Vector=Get gravity_override() -> (gravity_override:Vector)
    SimSpaceSettings:Sim Space Settings Structure=@n2
   ) -> (Pose) {
    bClampLinearTranslationLimitToRefPose=true
    bOverrideWorldGravity=true
    overridePhysicsAsset=&/F_MED_Painter_New_Phys
    simulationSpace=BaseBoneSpace
    alphaBoolBlend {blendOption=Linear}
    baseBoneRef {boneName=pelvis}
    componentAppliedLinearAccClamp=<2500 2500 2500>
    componentLinearAccScale=<0.4 0.3 0.1>
    componentLinearVelScale=<0.4 0.3 0.1>
    overrideWorldGravity {z=-980}
    simSpaceSettings {maxAngularAcceleration=1000, maxAngularVelocity=3, maxLinearAcceleration=250, maxLinearVelocity=500, velocityScaleZ=0, worldAlpha=1}
   }
   n4 = Component To Local(ComponentPose=@n3)
   n5 = RigidBody(ComponentPose=Local To Component(LocalPose=Use cached pose 'AnimCache'()), OverrideWorldGravity=Get gravity_override(), SimSpaceSettings=@n2) {
    bClampLinearTranslationLimitToRefPose=true
    bOverrideWorldGravity=true
    overridePhysicsAsset=&/F_MED_Painter_Frontend_Phys
    simulationSpace=BaseBoneSpace
    alphaBoolBlend {blendOption=Linear}
    baseBoneRef {boneName=pelvis}
    componentAppliedLinearAccClamp=<2500 2500 2500>
    componentLinearAccScale=<0.4 0.3 0.1>
    componentLinearVelScale=<0.4 0.3 0.1>
    overrideWorldGravity {z=-980}
    simSpaceSettings {maxAngularAcceleration=1000, maxAngularVelocity=3, maxLinearAcceleration=250, maxLinearVelocity=500, velocityScaleZ=0, worldAlpha=1}
   }
   n6 = Component To Local(ComponentPose=@n5)
   n7 = Get bIsInFrontEnd()
   n8 = "To Float (Boolean)"(InBool=Get bIsSkyDiving())
   n9 = "Transform (Modify) Bone - Bone: jacket_back_01_l"(
    ComponentPose=Local To Component(LocalPose=Blend Poses by bool(BlendPose_0=@n6, BlendPose_1=@n4, bActiveValue=@n7) {blendType=Linear})
    Alpha=@n8
    Translation="1,1.5,0"
    Rotation="0,0,10"
    Scale="1,1,1"
   ) {
    rotationMode=BMM_Additive
    translationMode=BMM_Additive
    alphaBoolBlend {blendOption=Linear}
    alphaScaleBiasClamp {bInterpResult=true}
    boneToModify {boneName=jacket_back_01_l}
    rotation {roll=10}
    translation=<1 1.5>
   }
   n10 = "Transform (Modify) Bone - Bone: jacket_back_01"(ComponentPose=@n9, Alpha=@n8, Translation="1,1.5,0", Rotation="0,0,10", Scale="1,1,1") {
    rotationMode=BMM_Additive
    translationMode=BMM_Additive
    alphaBoolBlend {blendOption=Linear}
    alphaScaleBiasClamp {bInterpResult=true}
    boneToModify {boneName=jacket_back_01}
    rotation {roll=10}
    translation=<1 1.5>
   }
   n11 = "Transform (Modify) Bone - Bone: jacket_back_01_r"(ComponentPose=@n10, Alpha=@n8, Translation="1,1.5,0", Rotation="0,0,10", Scale="1,1,1") {
    rotationMode=BMM_Additive
    translationMode=BMM_Additive
    alphaBoolBlend {blendOption=Linear}
    alphaScaleBiasClamp {bInterpResult=true}
    boneToModify {boneName=jacket_back_01_r}
    rotation {roll=10}
    translation=<1 1.5>
   }
   n12 = "Transform (Modify) Bone - Bone: jacket_side_01_l"(ComponentPose=@n11, Alpha=@n8, Translation="2,0,0", Rotation="7.5,0,7.5", Scale="1,1,1") {
    rotationMode=BMM_Additive
    translationMode=BMM_Additive
    alphaBoolBlend {blendOption=Linear}
    alphaScaleBiasClamp {bInterpResult=true}
    boneToModify {boneName=jacket_side_01_l}
    rotation {pitch=7.5, roll=7.5}
    translation {x=2}
   }
   n13 = "Transform (Modify) Bone - Bone: jacket_side_01_r"(ComponentPose=@n12, Alpha=@n8, Translation="-2,0,0", Rotation="-7.5,0,7.5", Scale="1,1,1") {
    rotationMode=BMM_Additive
    translationMode=BMM_Additive
    alphaBoolBlend {blendOption=Linear}
    alphaScaleBiasClamp {bInterpResult=true}
    boneToModify {boneName=jacket_side_01_r}
    rotation {pitch=-7.5, roll=7.5}
    translation {x=-2}
   }
   n14 = "To Float (Boolean)"(InBool=Get Alt Weapon() -> (Alt Weapon:bool))
   n15 = "Transform (Modify) Bone - Bone: jacket_frt_02_l"(ComponentPose=@n13, Alpha=@n14, Translation="3,0,0", Rotation="20,0,0", Scale="1,1,1") {
    rotationMode=BMM_Additive
    translationMode=BMM_Additive
    alphaBoolBlend {blendOption=Linear}
    alphaScaleBiasClamp {bInterpResult=true, interpSpeedIncreasing=5}
    boneToModify {boneName=jacket_frt_02_l}
    rotation {pitch=20}
    translation {x=3}
   }
   n16 = "Transform (Modify) Bone - Bone: jacket_frt_03_l"(ComponentPose=@n15, Alpha=@n14, Translation="0,0,0", Rotation="-20,20,0", Scale="1,1,1") {
    rotationMode=BMM_Additive
    translationMode=BMM_Additive
    alphaBoolBlend {blendOption=Linear}
    alphaScaleBiasClamp {bInterpResult=true, interpSpeedIncreasing=5}
    boneToModify {boneName=jacket_frt_03_l}
    rotation {pitch=-20, yaw=20}
   }
   n17 = "Transform (Modify) Bone - Bone: jacket_frt_04_l"(ComponentPose=@n16, Alpha=@n14, Translation="0,0,0", Rotation="-30,0,0", Scale="1,1,1") {
    rotationMode=BMM_Additive
    translationMode=BMM_Additive
    alphaBoolBlend {blendOption=Linear}
    alphaScaleBiasClamp {bInterpResult=true, interpSpeedIncreasing=5}
    boneToModify {boneName=jacket_frt_04_l}
    rotation {pitch=-30}
   }
   n18 = "Transform (Modify) Bone - Bone: jacket_frt_05_l"(ComponentPose=@n17, Alpha=@n14, Translation="0,0,0", Rotation="25,0,0", Scale="1,1,1") {
    rotationMode=BMM_Additive
    translationMode=BMM_Additive
    alphaBoolBlend {blendOption=Linear}
    alphaScaleBiasClamp {bInterpResult=true, interpSpeedIncreasing=5}
    boneToModify {boneName=jacket_frt_05_l}
    rotation {pitch=25}
   }
   n19 = "Transform (Modify) Bone - Bone: jacket_frt_06_l"(ComponentPose=@n18, Alpha=@n14, Translation="0,0,0", Rotation="15,0,0", Scale="1,1,1") {
    rotationMode=BMM_Additive
    translationMode=BMM_Additive
    alphaBoolBlend {blendOption=Linear}
    alphaScaleBiasClamp {bInterpResult=true, interpSpeedIncreasing=5}
    boneToModify {boneName=jacket_frt_06_l}
    rotation {pitch=15}
   }
   n20 = "Transform (Modify) Bone - Bone: jacket_frt_07_l"(ComponentPose=@n19, Alpha=@n14, Translation="0,0,0", Rotation="15,0,0", Scale="1,1,1") {
    rotationMode=BMM_Additive
    translationMode=BMM_Additive
    alphaBoolBlend {blendOption=Linear}
    alphaScaleBiasClamp {bInterpResult=true, interpSpeedIncreasing=5}
    boneToModify {boneName=jacket_frt_07_l}
    rotation {pitch=15}
   }
   n21 = "Transform (Modify) Bone - Bone: jacket_frtSide_01_l"(ComponentPose=@n20, Alpha=@n14, Translation="0,0,0", Rotation="10,20,5", Scale="1,1,1") {
    rotationMode=BMM_Additive
    translationMode=BMM_Additive
    alphaBoolBlend {blendOption=Linear}
    alphaScaleBiasClamp {bInterpResult=true, interpSpeedIncreasing=5}
    boneToModify {boneName=jacket_frtSide_01_l}
    rotation=<rot 10 20 5>
   }
   n22 = "Transform (Modify) Bone - Bone: jacket_side_01_l"(ComponentPose=@n21, Alpha=@n14, Translation="0,0,0", Rotation="15,0,0", Scale="1,1,1") {
    rotationMode=BMM_Additive
    translationMode=BMM_Additive
    alphaBoolBlend {blendOption=Linear}
    alphaScaleBiasClamp {bInterpResult=true, interpSpeedIncreasing=5}
    boneToModify {boneName=jacket_side_01_l}
    rotation {pitch=15}
   }
   n23 = "Constraint - jacket_frt_03_l (PPPPP)"!(
    ComponentPose=Local To Component(LocalPose=Use cached pose 'AnimCache'())
    Alpha=1
    ConstraintWeights_0=0.1633
    ConstraintWeights_1:float=0.3139
    ConstraintWeights_2:float=0.1437
    ConstraintWeights_3:float=0.3058
    ConstraintWeights_4:float=0.0732
   ) -> (Pose) {
    alphaBoolBlend {blendOption=Linear}
    boneToModify {boneName=jacket_frt_03_l}
    constraintSetup [
     {targetBone {boneName=spine_01}}
     {targetBone {boneName=spine_02}}
     {targetBone {boneName=spine_03}}
     {targetBone {boneName=spine_04}}
     {targetBone {boneName=pelvis}}
    ]
    constraintWeights [1,1,1,1,1]
   }
   n24 = "Constraint - jacket_frt_04_l (PP)"!(ComponentPose=@n23, Alpha=1, ConstraintWeights_0=0.7617, ConstraintWeights_1:float=0.2383) -> (Pose) {
    alphaBoolBlend {blendOption=Linear}
    boneToModify {boneName=jacket_frt_04_l}
    constraintSetup [{targetBone {boneName=pelvis}}, {targetBone {boneName=thigh_l}}]
    constraintWeights [1,1]
   }
   n25 = "Constraint - jacket_frt_05_l (PP)"!(ComponentPose=@n24, Alpha=1, ConstraintWeights_0=0.1011, ConstraintWeights_1:float=0.8989) -> (Pose) {
    alphaBoolBlend {blendOption=Linear}
    boneToModify {boneName=jacket_frt_05_l}
    constraintSetup [{targetBone {boneName=pelvis}}, {targetBone {boneName=thigh_l}}]
    constraintWeights [1,1]
   }
   n26 = "Constraint - jacket_frt_06_l (PP)"!(ComponentPose=@n25, Alpha=1, ConstraintWeights_0=0.0113, ConstraintWeights_1:float=0.9887) -> (Pose) {
    alphaBoolBlend {blendOption=Linear}
    boneToModify {boneName=jacket_frt_06_l}
    constraintSetup [{targetBone {boneName=pelvis}}, {targetBone {boneName=thigh_l}}]
    constraintWeights [1,1]
   }
   n27 = "Constraint - jacket_frt_07_l (PP)"!(ComponentPose=@n26, Alpha=1, ConstraintWeights_0=0.0191, ConstraintWeights_1:float=0.9809) -> (Pose) {
    alphaBoolBlend {blendOption=Linear}
    boneToModify {boneName=jacket_frt_07_l}
    constraintSetup [{targetBone {boneName=calf_l}}, {targetBone {boneName=thigh_l}}]
    constraintWeights [1,1]
   }
   n28 = "Constraint - jacket_frt_08_l (PP)"!(ComponentPose=@n27, Alpha=1, ConstraintWeights_0=0.963, ConstraintWeights_1:float=0.037) -> (Pose) {
    alphaBoolBlend {blendOption=Linear}
    boneToModify {boneName=jacket_frt_08_l}
    constraintSetup [{targetBone {boneName=calf_l}}, {targetBone {boneName=thigh_l}}]
    constraintWeights [1,1]
   }
   n29 = "Constraint - jacket_frt_09_l (P)"!(ComponentPose=@n28, Alpha=1, ConstraintWeights_0=1, ConstraintWeights_1:float=0) -> (Pose) {
    alphaBoolBlend {blendOption=Linear}
    boneToModify {boneName=jacket_frt_09_l}
    constraintSetup [{targetBone {boneName=calf_l}}]
    constraintWeights [1,0]
   }
   n30 = "Constraint - jacket_frt_03_r (PPPPP)"!(
    ComponentPose=@n29
    Alpha=1
    ConstraintWeights_0=0.1633
    ConstraintWeights_1:float=0.3137
    ConstraintWeights_2:float=0.1434
    ConstraintWeights_3:float=0.3059
    ConstraintWeights_4:float=0.0737
   ) -> (Pose) {
    alphaBoolBlend {blendOption=Linear}
    boneToModify {boneName=jacket_frt_03_r}
    constraintSetup [
     {targetBone {boneName=spine_01}}
     {targetBone {boneName=spine_02}}
     {targetBone {boneName=spine_03}}
     {targetBone {boneName=spine_04}}
     {targetBone {boneName=pelvis}}
    ]
    constraintWeights [1,1,1,1,1]
   }
   n31 = "Constraint - jacket_frt_04_r (PP)"!(ComponentPose=@n30, Alpha=1, ConstraintWeights_0=0.2391, ConstraintWeights_1:float=0.7609) -> (Pose) {
    alphaBoolBlend {blendOption=Linear}
    boneToModify {boneName=jacket_frt_04_r}
    constraintSetup [{targetBone {boneName=thigh_r}}, {targetBone {boneName=pelvis}}]
    constraintWeights [1,1]
   }
   n32 = "Constraint - jacket_frt_05_r (PP)"!(ComponentPose=@n31, Alpha=1, ConstraintWeights_0=0.9059, ConstraintWeights_1:float=0.0941) -> (Pose) {
    alphaBoolBlend {blendOption=Linear}
    boneToModify {boneName=jacket_frt_05_r}
    constraintSetup [{targetBone {boneName=thigh_r}}, {targetBone {boneName=pelvis}}]
    constraintWeights [1,1]
   }
   n33 = "Constraint - jacket_frt_06_r (P)"!(ComponentPose=@n32, Alpha=1, ConstraintWeights_0=1, ConstraintWeights_1:float=0) -> (Pose) {
    alphaBoolBlend {blendOption=Linear}
    boneToModify {boneName=jacket_frt_06_r}
    constraintSetup [{targetBone {boneName=thigh_r}}]
    constraintWeights [1,0]
   }
   n34 = "Constraint - jacket_frt_07_r (PP)"!(ComponentPose=@n33, Alpha=1, ConstraintWeights_0=0.9791, ConstraintWeights_1:float=0.0209) -> (Pose) {
    alphaBoolBlend {blendOption=Linear}
    boneToModify {boneName=jacket_frt_07_r}
    constraintSetup [{targetBone {boneName=thigh_r}}, {targetBone {boneName=calf_r}}]
    constraintWeights [1,1]
   }
   n35 = "Constraint - jacket_frt_08_r (PP)"!(ComponentPose=@n34, Alpha=1, ConstraintWeights_0=0.037, ConstraintWeights_1:float=0.963) -> (Pose) {
    alphaBoolBlend {blendOption=Linear}
    boneToModify {boneName=jacket_frt_08_r}
    constraintSetup [{targetBone {boneName=thigh_r}}, {targetBone {boneName=calf_r}}]
    constraintWeights [1,1]
   }
   n36 = "Constraint - jacket_frt_09_r (P)"!(ComponentPose=@n35, Alpha=1, ConstraintWeights_0=1, ConstraintWeights_1:float=0) -> (Pose) {
    alphaBoolBlend {blendOption=Linear}
    boneToModify {boneName=jacket_frt_09_r}
    constraintSetup [{targetBone {boneName=calf_r}}]
    constraintWeights [1,0]
   }
   n37 = "Constraint - jacket_frtSide_02_l (PPPPP)"!(
    ComponentPose=@n36
    Alpha=1
    ConstraintWeights_0=0.2331
    ConstraintWeights_1:float=0.3882
    ConstraintWeights_2:float=0.0505
    ConstraintWeights_3:float=0.1313
    ConstraintWeights_4:float=0.197
   ) -> (Pose) {
    alphaBoolBlend {blendOption=Linear}
    boneToModify {boneName=jacket_frtSide_02_l}
    constraintSetup [
     {targetBone {boneName=spine_01}}
     {targetBone {boneName=spine_02}}
     {targetBone {boneName=spine_03}}
     {targetBone {boneName=spine_04}}
     {targetBone {boneName=pelvis}}
    ]
    constraintWeights [1,1,1,1,1]
   }
   n38 = "Constraint - jacket_frtSide_03_l (PP)"!(ComponentPose=@n37, Alpha=1, ConstraintWeights_0=0.8376, ConstraintWeights_1:float=0.1624) -> (Pose) {
    alphaBoolBlend {blendOption=Linear}
    boneToModify {boneName=jacket_frtSide_03_l}
    constraintSetup [{targetBone {boneName=pelvis}}, {targetBone {boneName=thigh_l}}]
    constraintWeights [1,1]
   }
   n39 = "Constraint - jacket_frtSide_04_l (PP)"!(ComponentPose=@n38, Alpha=1, ConstraintWeights_0=0.1335, ConstraintWeights_1:float=0.8665) -> (Pose) {
    alphaBoolBlend {blendOption=Linear}
    boneToModify {boneName=jacket_frtSide_04_l}
    constraintSetup [{targetBone {boneName=pelvis}}, {targetBone {boneName=thigh_l}}]
    constraintWeights [1,1]
   }
   n40 = "Constraint - jacket_frtSide_05_l (PP)"!(ComponentPose=@n39, Alpha=1, ConstraintWeights_0=0.0134, ConstraintWeights_1:float=0.9866) -> (Pose) {
    alphaBoolBlend {blendOption=Linear}
    boneToModify {boneName=jacket_frtSide_05_l}
    constraintSetup [{targetBone {boneName=pelvis}}, {targetBone {boneName=thigh_l}}]
    constraintWeights [1,1]
   }
   n41 = "Constraint - jacket_frtSide_02_r (PPPPP)"!(
    ComponentPose=@n40
    Alpha=1
    ConstraintWeights_0=0.2289
    ConstraintWeights_1:float=0.3815
    ConstraintWeights_2:float=0.0525
    ConstraintWeights_3:float=0.1321
    ConstraintWeights_4:float=0.2049
   ) -> (Pose) {
    alphaBoolBlend {blendOption=Linear}
    boneToModify {boneName=jacket_frtSide_02_r}
    constraintSetup [
     {targetBone {boneName=spine_01}}
     {targetBone {boneName=spine_02}}
     {targetBone {boneName=spine_03}}
     {targetBone {boneName=spine_04}}
     {targetBone {boneName=pelvis}}
    ]
    constraintWeights [1,1,1,1,1]
   }
   n42 = "Constraint - jacket_frtSide_03_r (PP)"!(ComponentPose=@n41, Alpha=1, ConstraintWeights_0=0.1853, ConstraintWeights_1:float=0.8147) -> (Pose) {
    alphaBoolBlend {blendOption=Linear}
    boneToModify {boneName=jacket_frtSide_03_r}
    constraintSetup [{targetBone {boneName=thigh_r}}, {targetBone {boneName=pelvis}}]
    constraintWeights [1,1]
   }
   n43 = "Constraint - jacket_frtSide_04_r (PP)"!(ComponentPose=@n42, Alpha=1, ConstraintWeights_0=0.8764, ConstraintWeights_1:float=0.1236) -> (Pose) {
    alphaBoolBlend {blendOption=Linear}
    boneToModify {boneName=jacket_frtSide_04_r}
    constraintSetup [{targetBone {boneName=thigh_r}}, {targetBone {boneName=pelvis}}]
    constraintWeights [1,1]
   }
   n44 = "Constraint - jacket_frtSide_05_r (P)"!(ComponentPose=@n43, Alpha=1, ConstraintWeights_0=1, ConstraintWeights_1:float=0) -> (Pose) {
    alphaBoolBlend {blendOption=Linear}
    boneToModify {boneName=jacket_frtSide_05_r}
    constraintSetup [{targetBone {boneName=thigh_r}}]
    constraintWeights [1,0]
   }
   n45 = "Constraint - jacket_side_02_l (PPPPP)"!(
    ComponentPose=@n44
    Alpha=1
    ConstraintWeights_0=0.2426
    ConstraintWeights_1:float=0.334
    ConstraintWeights_2:float=0.0344
    ConstraintWeights_3:float=0.0663
    ConstraintWeights_4:float=0.3227
   ) -> (Pose) {
    alphaBoolBlend {blendOption=Linear}
    boneToModify {boneName=jacket_side_02_l}
    constraintSetup [
     {targetBone {boneName=spine_01}}
     {targetBone {boneName=spine_02}}
     {targetBone {boneName=spine_03}}
     {targetBone {boneName=spine_04}}
     {targetBone {boneName=pelvis}}
    ]
    constraintWeights [1,1,1,1,1]
   }
   n46 = "Constraint - jacket_side_03_l (PP)"!(ComponentPose=@n45, Alpha=1, ConstraintWeights_0=0.8489, ConstraintWeights_1:float=0.1511) -> (Pose) {
    alphaBoolBlend {blendOption=Linear}
    boneToModify {boneName=jacket_side_03_l}
    constraintSetup [{targetBone {boneName=pelvis}}, {targetBone {boneName=thigh_l}}]
    constraintWeights [1,1]
   }
   n47 = "Constraint - jacket_side_04_l (PP)"!(ComponentPose=@n46, Alpha=1, ConstraintWeights_0=0.3356, ConstraintWeights_1:float=0.6644) -> (Pose) {
    alphaBoolBlend {blendOption=Linear}
    boneToModify {boneName=jacket_side_04_l}
    constraintSetup [{targetBone {boneName=pelvis}}, {targetBone {boneName=thigh_l}}]
    constraintWeights [1,1]
   }
   n48 = "Constraint - jacket_side_05_l (PP)"!(ComponentPose=@n47, Alpha=1, ConstraintWeights_0=0.0309, ConstraintWeights_1:float=0.9691) -> (Pose) {
    alphaBoolBlend {blendOption=Linear}
    boneToModify {boneName=jacket_side_05_l}
    constraintSetup [{targetBone {boneName=pelvis}}, {targetBone {boneName=thigh_l}}]
    constraintWeights [1,1]
   }
   n49 = "Constraint - jacket_side_06_l (P)"!(ComponentPose=@n48, Alpha=1, ConstraintWeights_0=1, ConstraintWeights_1:float=0) -> (Pose) {
    alphaBoolBlend {blendOption=Linear}
    boneToModify {boneName=jacket_side_06_l}
    constraintSetup [{targetBone {boneName=thigh_l}}]
    constraintWeights [1,0]
   }
   n50 = "Constraint - jacket_side_07_l (PP)"!(ComponentPose=@n49, Alpha=1, ConstraintWeights_0=0.3489, ConstraintWeights_1:float=0.6511) -> (Pose) {
    alphaBoolBlend {blendOption=Linear}
    boneToModify {boneName=jacket_side_07_l}
    constraintSetup [{targetBone {boneName=calf_l}}, {targetBone {boneName=thigh_l}}]
    constraintWeights [1,1]
   }
   n51 = "Constraint - jacket_side_08_l (PP)"!(ComponentPose=@n50, Alpha=1, ConstraintWeights_0=0.974, ConstraintWeights_1:float=0.026) -> (Pose) {
    alphaBoolBlend {blendOption=Linear}
    boneToModify {boneName=jacket_side_08_l}
    constraintSetup [{targetBone {boneName=calf_l}}, {targetBone {boneName=thigh_l}}]
    constraintWeights [1,1]
   }
   n52 = "Constraint - jacket_side_02_r (PPPPP)"!(
    ComponentPose=@n51
    Alpha=1
    ConstraintWeights_0=0.2507
    ConstraintWeights_1:float=0.3439
    ConstraintWeights_2:float=0.0296
    ConstraintWeights_3:float=0.0635
    ConstraintWeights_4:float=0.3122
   ) -> (Pose) {
    alphaBoolBlend {blendOption=Linear}
    boneToModify {boneName=jacket_side_02_r}
    constraintSetup [
     {targetBone {boneName=spine_01}}
     {targetBone {boneName=spine_02}}
     {targetBone {boneName=spine_03}}
     {targetBone {boneName=spine_04}}
     {targetBone {boneName=pelvis}}
    ]
    constraintWeights [1,1,1,1,1]
   }
   n53 = "Constraint - jacket_side_03_r (PP)"!(ComponentPose=@n52, Alpha=1, ConstraintWeights_0=0.1511, ConstraintWeights_1:float=0.8489) -> (Pose) {
    alphaBoolBlend {blendOption=Linear}
    boneToModify {boneName=jacket_side_03_r}
    constraintSetup [{targetBone {boneName=thigh_r}}, {targetBone {boneName=pelvis}}]
    constraintWeights [1,1]
   }
   n54 = "Constraint - jacket_side_04_r (PP)"!(ComponentPose=@n53, Alpha=1, ConstraintWeights_0=0.6682, ConstraintWeights_1:float=0.3318) -> (Pose) {
    alphaBoolBlend {blendOption=Linear}
    boneToModify {boneName=jacket_side_04_r}
    constraintSetup [{targetBone {boneName=thigh_r}}, {targetBone {boneName=pelvis}}]
    constraintWeights [1,1]
   }
   n55 = "Constraint - jacket_side_05_r (PP)"!(ComponentPose=@n54, Alpha=1, ConstraintWeights_0=0.9865, ConstraintWeights_1:float=0.0135) -> (Pose) {
    alphaBoolBlend {blendOption=Linear}
    boneToModify {boneName=jacket_side_05_r}
    constraintSetup [{targetBone {boneName=thigh_r}}, {targetBone {boneName=pelvis}}]
    constraintWeights [1,1]
   }
   n56 = "Constraint - jacket_side_06_r (P)"!(ComponentPose=@n55, Alpha=1, ConstraintWeights_0=1, ConstraintWeights_1:float=0) -> (Pose) {
    alphaBoolBlend {blendOption=Linear}
    boneToModify {boneName=jacket_side_06_r}
    constraintSetup [{targetBone {boneName=thigh_r}}]
    constraintWeights [1,0]
   }
   n57 = "Constraint - jacket_side_07_r (PP)"!(ComponentPose=@n56, Alpha=1, ConstraintWeights_0=0.6415, ConstraintWeights_1:float=0.3584) -> (Pose) {
    alphaBoolBlend {blendOption=Linear}
    boneToModify {boneName=jacket_side_07_r}
    constraintSetup [{targetBone {boneName=thigh_r}}, {targetBone {boneName=calf_r}}]
    constraintWeights [1,1]
   }
   n58 = "Constraint - jacket_side_08_r (PP)"!(ComponentPose=@n57, Alpha=1, ConstraintWeights_0=0.026, ConstraintWeights_1:float=0.974) -> (Pose) {
    alphaBoolBlend {blendOption=Linear}
    boneToModify {boneName=jacket_side_08_r}
    constraintSetup [{targetBone {boneName=thigh_r}}, {targetBone {boneName=calf_r}}]
    constraintWeights [1,1]
   }
   n59 = "Constraint - jacket_back_02_l (PPPPP)"!(
    ComponentPose=@n58
    Alpha=1
    ConstraintWeights_0=0.2786
    ConstraintWeights_1:float=0.3325
    ConstraintWeights_2:float=0.0691
    ConstraintWeights_3:float=0.026
    ConstraintWeights_4:float=0.2939
   ) -> (Pose) {
    alphaBoolBlend {blendOption=Linear}
    boneToModify {boneName=jacket_back_02_l}
    constraintSetup [
     {targetBone {boneName=spine_01}}
     {targetBone {boneName=spine_02}}
     {targetBone {boneName=spine_03}}
     {targetBone {boneName=spine_04}}
     {targetBone {boneName=pelvis}}
    ]
    constraintWeights [1,1,1,1,1]
   }
   n60 = "Constraint - jacket_back_03_l (PP)"!(ComponentPose=@n59, Alpha=1, ConstraintWeights_0=0.9601, ConstraintWeights_1:float=0.0399) -> (Pose) {
    alphaBoolBlend {blendOption=Linear}
    boneToModify {boneName=jacket_back_03_l}
    constraintSetup [{targetBone {boneName=pelvis}}, {targetBone {boneName=thigh_l}}]
    constraintWeights [1,1]
   }
   n61 = "Constraint - jacket_back_04_l (PP)"!(ComponentPose=@n60, Alpha=1, ConstraintWeights_0=0.5447, ConstraintWeights_1:float=0.4553) -> (Pose) {
    alphaBoolBlend {blendOption=Linear}
    boneToModify {boneName=jacket_back_04_l}
    constraintSetup [{targetBone {boneName=pelvis}}, {targetBone {boneName=thigh_l}}]
    constraintWeights [1,1]
   }
   n62 = "Constraint - jacket_back_05_l (PP)"!(ComponentPose=@n61, Alpha=1, ConstraintWeights_0=0.0682, ConstraintWeights_1:float=0.9318) -> (Pose) {
    alphaBoolBlend {blendOption=Linear}
    boneToModify {boneName=jacket_back_05_l}
    constraintSetup [{targetBone {boneName=pelvis}}, {targetBone {boneName=thigh_l}}]
    constraintWeights [1,1]
   }
   n63 = "Constraint - jacket_back_02_r (PPPPP)"!(
    ComponentPose=@n62
    Alpha=1
    ConstraintWeights_0=0.283
    ConstraintWeights_1:float=0.3464
    ConstraintWeights_2:float=0.072
    ConstraintWeights_3:float=0.0211
    ConstraintWeights_4:float=0.2775
   ) -> (Pose) {
    alphaBoolBlend {blendOption=Linear}
    boneToModify {boneName=jacket_back_02_r}
    constraintSetup [
     {targetBone {boneName=spine_01}}
     {targetBone {boneName=spine_02}}
     {targetBone {boneName=spine_03}}
     {targetBone {boneName=spine_04}}
     {targetBone {boneName=pelvis}}
    ]
    constraintWeights [1,1,1,1,1]
   }
   n64 = "Constraint - jacket_back_03_r (PP)"!(ComponentPose=@n63, Alpha=1, ConstraintWeights_0=0.0399, ConstraintWeights_1:float=0.9601) -> (Pose) {
    alphaBoolBlend {blendOption=Linear}
    boneToModify {boneName=jacket_back_03_r}
    constraintSetup [{targetBone {boneName=thigh_r}}, {targetBone {boneName=pelvis}}]
    constraintWeights [1,1]
   }
   n65 = "Constraint - jacket_back_04_r (PP)"!(ComponentPose=@n64, Alpha=1, ConstraintWeights_0=0.4565, ConstraintWeights_1:float=0.5435) -> (Pose) {
    alphaBoolBlend {blendOption=Linear}
    boneToModify {boneName=jacket_back_04_r}
    constraintSetup [{targetBone {boneName=thigh_r}}, {targetBone {boneName=pelvis}}]
    constraintWeights [1,1]
   }
   n66 = "Constraint - jacket_back_05_r (PP)"!(ComponentPose=@n65, Alpha=1, ConstraintWeights_0=0.9486, ConstraintWeights_1:float=0.0514) -> (Pose) {
    alphaBoolBlend {blendOption=Linear}
    boneToModify {boneName=jacket_back_05_r}
    constraintSetup [{targetBone {boneName=thigh_r}}, {targetBone {boneName=pelvis}}]
    constraintWeights [1,1]
   }
   n67 = "Constraint - jacket_back_02 (PPPPP)"!(
    ComponentPose=@n66
    Alpha=1
    ConstraintWeights_0=0.2641
    ConstraintWeights_1:float=0.3737
    ConstraintWeights_2:float=0.1147
    ConstraintWeights_3:float=0.0276
    ConstraintWeights_4:float=0.2199
   ) -> (Pose) {
    alphaBoolBlend {blendOption=Linear}
    boneToModify {boneName=jacket_back_02}
    constraintSetup [
     {targetBone {boneName=spine_01}}
     {targetBone {boneName=spine_02}}
     {targetBone {boneName=spine_03}}
     {targetBone {boneName=spine_04}}
     {targetBone {boneName=pelvis}}
    ]
    constraintWeights [1,1,1,1,1]
   }
   n68 = "Constraint - jacket_back_03 (P)"!(ComponentPose=@n67, Alpha=1, ConstraintWeights_0=1, ConstraintWeights_1:float=0) -> (Pose) {
    alphaBoolBlend {blendOption=Linear}
    boneToModify {boneName=jacket_back_03}
    constraintSetup [{targetBone {boneName=pelvis}}]
    constraintWeights [1,0]
   }
   n69 = "Constraint - jacket_back_04 (PPP)"!(ComponentPose=@n68, Alpha=1, ConstraintWeights_0=0.1801, ConstraintWeights_1:float=0.7586, ConstraintWeights_2:float=0.0613) -> (Pose) {
    alphaBoolBlend {blendOption=Linear}
    boneToModify {boneName=jacket_back_04}
    constraintSetup [{targetBone {boneName=thigh_r}}, {targetBone {boneName=pelvis}}, {targetBone {boneName=thigh_l}}]
    constraintWeights [1,1,1]
   }
   n70 = "Constraint - jacket_back_05 (PP)"!(ComponentPose=@n69, Alpha=1, ConstraintWeights_0=0.0699, ConstraintWeights_1:float=0.9301) -> (Pose) {
    alphaBoolBlend {blendOption=Linear}
    boneToModify {boneName=jacket_back_05}
    constraintSetup [{targetBone {boneName=pelvis}}, {targetBone {boneName=thigh_l}}]
    constraintWeights [1,1]
   }
   Output Pose(Result=Blend Poses by bool(
     BlendPose_0=Component To Local(ComponentPose=@n70)
     BlendPose_1=Component To Local(ComponentPose=@n22)
     bActiveValue=Get rigids_off() -> (rigids_off:bool)
    ) {blendType=Linear}) {layerGroup=None}
   n71 = Anim Dynamics - Bone: hairA_01_l(ComponentPose=Local To Component(LocalPose=Use cached pose 'StartCache'()), Alpha=0.25, ExternalForce="0,0,0") {
    angularDampingOverride=1
    angularSpringConstant=30000
    bAngularSpring=true
    bGravityOverrideInSimSpace=true
    bOverrideAngularDamping=true
    gravityScale=0.37
    simulationSpace=RootRelative
    alphaBoolBlend {blendOption=Linear}
    boundBone {boneName=hairA_01_l}
    physicsBodyDefinitions [{
      sphereCollisionRadius=0
      boundBone {boneName=hairA_01_l}
      boxExtents=<10 20 20>
      constraintSetup {angularLimitsMax=<5 5 25>, angularLimitsMin=<-5 -5 -4>, angularTarget=<0 0 0>}
      localJointOffset=<5 0 0>
     }]
   }
   n72 = Anim Dynamics - Bone: hairA_02_l(ComponentPose=@n71, Alpha=0.35, ExternalForce="0,0,0") {
    angularDampingOverride=1
    angularSpringConstant=30000
    bAngularSpring=true
    bGravityOverrideInSimSpace=true
    bOverrideAngularDamping=true
    gravityScale=0.37
    simulationSpace=RootRelative
    alphaBoolBlend {blendOption=Linear}
    boundBone {boneName=hairA_02_l}
    physicsBodyDefinitions [{
      sphereCollisionRadius=0
      boundBone {boneName=hairA_02_l}
      boxExtents=<10 20 20>
      constraintSetup {angularLimitsMax=<6 6 35>, angularLimitsMin=<-5 -15 -10>, angularTarget=<0 0 0>}
      localJointOffset=<5 0 0>
     }]
   }
   n73 = Anim Dynamics - Bone: hairA_03_l(ComponentPose=@n72, Alpha=0.65, ExternalForce="0,0,0") {
    angularDampingOverride=1
    angularSpringConstant=30000
    bAngularSpring=true
    bGravityOverrideInSimSpace=true
    bOverrideAngularDamping=true
    gravityScale=0.37
    simulationSpace=RootRelative
    alphaBoolBlend {blendOption=Linear}
    boundBone {boneName=hairA_03_l}
    physicsBodyDefinitions [{
      sphereCollisionRadius=0
      boundBone {boneName=hairA_03_l}
      boxExtents=<10 20 20>
      constraintSetup {angularLimitsMax=<6 10 45>, angularLimitsMin=<-5 -15 0>, angularTarget=<0 0 0>}
      localJointOffset=<5 0 0>
     }]
   }
   n74 = Anim Dynamics - Bone: hairA_01_r(ComponentPose=@n73, Alpha=0.35, ExternalForce="0,0,0") {
    angularDampingOverride=1
    angularSpringConstant=300
    bAngularSpring=true
    bGravityOverrideInSimSpace=true
    bOverrideAngularDamping=true
    gravityScale=0.32
    simulationSpace=RootRelative
    alphaBoolBlend {blendOption=Linear}
    boundBone {boneName=hairA_01_r}
    physicsBodyDefinitions [{
      sphereCollisionRadius=0
      boundBone {boneName=hairA_01_r}
      boxExtents=<10 15 25>
      constraintSetup {angularTargetAxis=AxisY, angularLimitsMax=<0 0 0>, angularLimitsMin=<0 0 -50>, angularTarget=<0 1 0>}
      localJointOffset=<10 0 0>
     }]
   }
   n75 = Anim Dynamics - Bone: hairA_02_r(ComponentPose=@n74, Alpha=0.25, ExternalForce="0,0,0") {
    angularDampingOverride=1
    angularSpringConstant=1500
    bAngularSpring=true
    bGravityOverrideInSimSpace=true
    bOverrideAngularDamping=true
    gravityScale=0.32
    simulationSpace=RootRelative
    alphaBoolBlend {blendOption=Linear}
    boundBone {boneName=hairA_02_r}
    physicsBodyDefinitions [{
      sphereCollisionRadius=0
      boundBone {boneName=hairA_02_r}
      boxExtents=<10 15 25>
      constraintSetup {angularLimitsMax=<0 0 1>, angularLimitsMin=<0 0 -30>, angularTarget=<0 0 0>}
      localJointOffset=<10 0 0>
     }]
   }
   n76 = Anim Dynamics - Bone: hairA_03_r(ComponentPose=@n75, Alpha=0.1, ExternalForce="0,0,0") {
    angularDampingOverride=1
    angularSpringConstant=1500
    bAngularSpring=true
    bGravityOverrideInSimSpace=true
    bOverrideAngularDamping=true
    gravityScale=0.32
    simulationSpace=RootRelative
    alphaBoolBlend {blendOption=Linear}
    boundBone {boneName=hairA_03_r}
    physicsBodyDefinitions [{
      sphereCollisionRadius=0
      boundBone {boneName=hairA_03_r}
      boxExtents=<10 15 25>
      constraintSetup {angularLimitsMax=<0 0 2>, angularLimitsMin=<0 0 -10>, angularTarget=<0 0 0>}
      localJointOffset=<10 0 0>
     }]
   }
   n77 = Anim Dynamics - Bone: hairB_01_l(ComponentPose=@n76, Alpha=1, ExternalForce="0,0,0") {
    angularDampingOverride=1
    angularSpringConstant=1500
    bAngularSpring=true
    bGravityOverrideInSimSpace=true
    bOverrideAngularDamping=true
    gravityScale=0.35
    simulationSpace=RootRelative
    alphaBoolBlend {blendOption=Linear}
    boundBone {boneName=hairB_01_l}
    physicsBodyDefinitions [{
      sphereCollisionRadius=0
      boundBone {boneName=hairB_01_l}
      boxExtents=<15 25 25>
      constraintSetup {angularLimitsMax=<5 5 25>, angularLimitsMin=<-5 -5 -20>, angularTarget=<0 0 0>}
      localJointOffset=<10 0 0>
     }]
   }
   n78 = Anim Dynamics - Bone: hairB_02_l(ComponentPose=@n77, Alpha=1, ExternalForce="0,0,0") {
    angularDampingOverride=1
    angularSpringConstant=1500
    bAngularSpring=true
    bGravityOverrideInSimSpace=true
    bOverrideAngularDamping=true
    gravityScale=0.35
    simulationSpace=RootRelative
    alphaBoolBlend {blendOption=Linear}
    boundBone {boneName=hairB_02_l}
    physicsBodyDefinitions [{
      sphereCollisionRadius=0
      boundBone {boneName=hairB_02_l}
      boxExtents=<15 25 25>
      constraintSetup {angularLimitsMax=<5 5 25>, angularLimitsMin=<-5 -5 -20>, angularTarget=<0 0 0>}
      localJointOffset=<10 0 0>
     }]
   }
   n79 = Anim Dynamics - Bone: hairB_a01_l(ComponentPose=@n78, Alpha=1, ExternalForce="0,0,0") {
    angularDampingOverride=1
    angularSpringConstant=1500
    bAngularSpring=true
    bGravityOverrideInSimSpace=true
    bOverrideAngularDamping=true
    gravityScale=0.35
    simulationSpace=RootRelative
    alphaBoolBlend {blendOption=Linear}
    boundBone {boneName=hairB_a01_l}
    physicsBodyDefinitions [{
      sphereCollisionRadius=0
      boundBone {boneName=hairB_a01_l}
      boxExtents=<5 15 15>
      constraintSetup {angularLimitsMax=<5 5 25>, angularLimitsMin=<-5 -5 -20>, angularTarget=<0 0 0>}
      localJointOffset=<4 0 0>
     }]
   }
   n80 = Anim Dynamics - Bone: hairB_01_r(ComponentPose=@n79, Alpha=0.7, ExternalForce="0,0,0") {
    angularDampingOverride=0.9999
    angularSpringConstant=1500
    bAngularSpring=true
    bGravityOverrideInSimSpace=true
    bOverrideAngularDamping=true
    gravityScale=3.1
    simulationSpace=RootRelative
    alphaBoolBlend {blendOption=Linear}
    boundBone {boneName=hairB_01_r}
    physicsBodyDefinitions [{
      sphereCollisionRadius=0
      boundBone {boneName=hairB_01_r}
      boxExtents=<8 28 28>
      constraintSetup {angularLimitsMax=<0 0 0>, angularLimitsMin=<0 0 -40>, angularTarget=<0 0 0>}
      localJointOffset=<5 0 0>
     }]
   }
   n81 = Anim Dynamics - Bone: hairB_02_r(ComponentPose=@n80, Alpha=0.5, ExternalForce="0,0,0") {
    angularDampingOverride=0.9999
    angularSpringConstant=1500
    bAngularSpring=true
    bGravityOverrideInSimSpace=true
    bOverrideAngularDamping=true
    gravityScale=3.1
    simulationSpace=RootRelative
    alphaBoolBlend {blendOption=Linear}
    boundBone {boneName=hairB_02_r}
    physicsBodyDefinitions [{
      sphereCollisionRadius=0
      boundBone {boneName=hairB_02_r}
      boxExtents=<8 28 28>
      constraintSetup {angularLimitsMax=<4 8 25>, angularLimitsMin=<-4 -8 -25>, angularTarget=<0 0 0>}
      localJointOffset=<5 0 0>
     }]
   }
   n82 = Anim Dynamics - Bone: hairB_a01_r(ComponentPose=@n81, Alpha=0.356, ExternalForce="0,0,0") {
    angularDampingOverride=0.9999
    angularSpringConstant=1500
    bAngularSpring=true
    bGravityOverrideInSimSpace=true
    bOverrideAngularDamping=true
    gravityScale=3.1
    simulationSpace=RootRelative
    alphaBoolBlend {blendOption=Linear}
    boundBone {boneName=hairB_a01_r}
    physicsBodyDefinitions [{
      sphereCollisionRadius=0
      boundBone {boneName=hairB_a01_r}
      boxExtents=<8 28 28>
      constraintSetup {angularLimitsMax=<0 0 0>, angularLimitsMin=<0 0 -40>, angularTarget=<0 0 0>}
      localJointOffset=<5 0 0>
     }]
   }
   n83 = "Transform (Modify) Bone - Bone: hood_offset"(ComponentPose=@n82, Alpha=Get head_rot_back() -> (head_rot_back:double), Translation="0.000,-10,0.000", Rotation="0,0,0", Scale="1,1,1") {
    rotationMode=BMM_Additive
    rotationSpace=BCS_BoneSpace
    translationMode=BMM_Additive
    translationSpace=BCS_BoneSpace
    alphaBoolBlend {blendOption=Linear}
    boneToModify {boneName=hood_offset}
    translation {y=-10}
   }
   n84 = "Transform (Modify) Bone - Bone: jacket_base"(ComponentPose=@n83, Alpha=Get spine_rot_back() -> (spine_rot_back:double), Translation="0,0,0", Rotation="0,-40,0.000", Scale="1,1,1") {
    rotationMode=BMM_Additive
    rotationSpace=BCS_BoneSpace
    translationMode=BMM_Additive
    translationSpace=BCS_BoneSpace
    alphaBoolBlend {blendOption=Linear}
    boneToModify {boneName=jacket_base}
    rotation {yaw=-40}
   }
   Save cached pose 'AnimCache'(Pose=Component To Local(ComponentPose=@n84)) {cacheName=AnimCache}
   Save cached pose 'StartCache'(Pose=Copy Pose From Mesh() {bUseAttachedParent=true}) {cacheName=StartCache}
  ]}
 EventGraph {nodes [
   Event BlueprintUpdateAnimation()
   > n1 = Sequence!(execute) -> (then_0, then_1, then_2:Exec, then_3:Exec, then_4:Exec, then_5:Exec, then_6:Exec)
   n2 = Gravity Override State Function DEPRECATED(execute=@n1.then_1, CharacterPartAnimInstance=Self-Reference(), GameTimeInSeconds=GetGameTimeInSeconds())
   > Set gravity_override(gravity_override=@n2)
   n3 = Branch(execute=@n1.then_5, Condition=Get bIsInFrontEnd())
   > Set Sim Space Settings(Sim Space Settings:Sim Space Settings Structure=Get Sim Space Settings_FrontEnd() -> (Sim Space Settings_FrontEnd:Sim Space Settings Structure)) -> (Output_Get:Sim Space Settings Structure)
   Set Sim Space Settings(execute=@n3.else, Sim Space Settings=Get Sim Space Settings_InZone() -> (Sim Space Settings_InZone:Sim Space Settings Structure))
   n4 = BreakTransform(InTransform=GetDeltaTransformFromRefPose(self=GetOwnerSkeletalMeshComponent(), BoneName=head, BaseName=spine_05))
   Set head_rot_back(execute=@n1.then_0, head_rot_back:double=MapRangeClamped(Value=BreakRotator(InRot=@n4.Rotation).Yaw, InRangeA=0, InRangeB=-55)) -> (Output_Get:double)
   n5 = Get CurrentLOD()
   n6 = Get bRigidBodyEnabled()
   Set rigids_off(execute=@n1.then_3, rigids_off:bool=OR Boolean(A=NOT Boolean(A=@n6), B="integer > integer"(A=@n5, B=1) {is=title K2Node_CallFunction})) -> (Output_Get:bool)
   n7 = BreakTransform(InTransform=GetDeltaTransformFromRefPose(self=GetOwnerSkeletalMeshComponent(), BoneName=spine_05, BaseName=pelvis))
   Set spine_rot_back(execute=@n1.then_2, spine_rot_back:double=MapRangeClamped(Value=BreakRotator(InRot=@n7.Rotation).Yaw, InRangeA=25, InRangeB=52)) -> (Output_Get:double)
   Set Alt Weapon(execute=@n1.then_4, Alt Weapon:bool=false) -> (Output_Get:bool)
   > n8 = Cast To FortPlayerAnimInstance(Object=GetAnimInstance(self=GetOwnerSkeletalMeshComponent()))
   n9 = Switch on EFortWeaponCoreAnimation!(execute=[@n8.then, @n13.then], Selection:EFortWeaponCoreAnimation Enum=Get WeaponCoreAnim(self=@n8)) -> (
    Melee:Exec
    Pistol:Exec
    Shotgun:Exec
    PaperBlueprint:Exec
    Rifle:Exec
    MeleeOneHand:Exec
    MachinePistol:Exec
    RocketLauncher:Exec
    GrenadeLauncher:Exec
    GoingCommando:Exec
    AssaultRifle:Exec
    TacticalShotgun:Exec
    SniperRifle:Exec
    TrapPlacement:Exec
    ShoulderLauncher:Exec
    AbilityDecoTool:Exec
    Crossbow:Exec
    C4:Exec
    RemoteControl:Exec
    DualWield:Exec
    AR_BullPup:Exec
    AR_ForwardGrip:Exec
    MedPackPaddles:Exec
    SMG_P90:Exec
    AR_DrumGun:Exec
    Consumable_Small:Exec
    Consumable_Large:Exec
    Balloon:Exec
    MountedTurret:Exec
    CreativeTool:Exec
    ExplosiveBow:Exec
    AshtonIndigo:Exec
    AshtonChicago:Exec
    MeleeDualWield:Exec
    Unarmed:Exec
   )
   Branch(execute=[@n9.AssaultRifle, @n9.Rifle, @n9.Shotgun, @n9.TacticalShotgun], Condition=Get bIsCrouching())
   > Set Alt Weapon(Alt Weapon=true)
   Event OnParachuteOpened()
   > n10 = Delay(Duration=1.5)
   ResetDynamics(execute=[@n10.then, @n11.then, @n12.then], InTeleportType=TeleportPhysics)
   > ResetDynamics(InTeleportType=ResetPhysics)
   n11 = Event OnEndSkydiving()
   n12 = Event OnBeginSkydiving()
   n13 = Set Alt Weapon(Alt Weapon=false)
   TryGetPawnOwner()
   Reroute Node(InputPin:Wildcard) -> (OutputPin:Wildcard)
  ]}
]
variables [
 Alt Weapon {type=boolean, default=true}
 gravity_override {type=Vector}
 head_rot_back {type=number}
 rigids_off {type=boolean}
 Sim Space Settings {
  type=SimSpaceSettings
  default {
   dampingAlpha=1
   maxAngularAcceleration=1000
   maxAngularVelocity=3
   maxLinearAcceleration=250
   maxLinearVelocity=500
   velocityScaleZ=0
   worldAlpha=1
   externalLinearDragV=<0 0 0>
  }
 }
 Sim Space Settings_FrontEnd {
  type=SimSpaceSettings
  default {
   dampingAlpha=1
   maxAngularAcceleration=10000
   maxAngularVelocity=10
   maxLinearAcceleration=250
   maxLinearVelocity=500
   velocityScaleZ=0
   worldAlpha=1
   externalLinearDragV=<0 0 0>
  }
 }
 Sim Space Settings_InZone {
  type=SimSpaceSettings
  default {
   dampingAlpha=1
   maxAngularAcceleration=1000
   maxAngularVelocity=3
   maxLinearAcceleration=250
   maxLinearVelocity=500
   velocityScaleZ=0
   worldAlpha=1
   externalLinearDragV=<0 0 0>
  }
 }
 spine_rot_back {type=number}
]
