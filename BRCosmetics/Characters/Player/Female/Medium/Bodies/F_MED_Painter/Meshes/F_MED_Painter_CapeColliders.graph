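# Collision-only physics asset for the F_MED_Painter character body, named as the colliders for its cape.
# Capsules cover the head, clavicles, feet, pelvis and spine, with a second, rotated capsule on spine_02 through spine_04.
# Named capsules on back, side and front jacket bones add collision volume that follows the jacket.
# Arm, thigh, calf and neck bodies have no shapes, and no constraints join any bodies.
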PhysicsAsset /BRCosmetics/Characters/Player/Female/Medium/Bodies/F_MED_Painter/Meshes/F_MED_Painter_CapeColliders
bodies {
 calf_l []
 calf_r []
 clavicle_l [{shapeType=Capsule, length=7.6964, radius=6.3853, center=<4.0147 -0.8354 -1.7419>, rotation=<rot -77.5566 177.3915 7.7314>}]
 clavicle_r [{shapeType=Capsule, length=7.6964, radius=8.2957, center=<-4.0147 0.8354 1.7419>, rotation=<rot -77.5565 177.3915 -172.2685>}]
 foot_l [{shapeType=Capsule, length=12.5982, radius=6.5127, center=<-2.2072 6.0534 -0.3367>, rotation=<rot 0.0388 1.4701 -89.3982>}]
 foot_r [{shapeType=Capsule, length=12.5982, radius=6.5127, center=<2.2072 -6.0534 0.3367>, rotation=<rot 0.0388 1.4701 90.6018>}]
 head [{shapeType=Capsule, length=0, radius=18, center=<9.0697 0.0746 0>, rotation=<rot 0 -90 0>}]
 jacket_back_05 [{shapeType=Capsule, length=10.75, radius=14, shapeName=jacket_back_05_capsule, center=<2.3578 -7.5378 0>, rotation=<rot 0 -179.9994 180>}]
 jacket_frtSide_05_l [{
   shapeType=Capsule
   length=19
   radius=11.25
   shapeName=jacket_frtSide_05_l_capsule
   center=<1.7891 -1.0341 2.9977>
   rotation=<rot -90 -179.789 179.7891>
  }]
 jacket_frtSide_05_r [{
   shapeType=Capsule
   length=19
   radius=11.25
   shapeName=jacket_frtSide_05_r_capsule
   center=<1.7891 -1.0341 -4.6631>
   rotation=<rot -90 166.7373 -166.7373>
  }]
 jacket_frt_08_l [{shapeType=Capsule, length=18.5, radius=5.5, shapeName=jacket_frt_08_l_capsule, rotation=<rot -90 -172.7217 172.7218>}]
 jacket_frt_08_r [{shapeType=Capsule, length=18.5, radius=5.5, shapeName=jacket_frt_08_r_capsule, rotation=<rot -90 7.2783 -7.2783>}]
 jacket_side_08_l [{shapeType=Capsule, length=30.25, radius=9.75, shapeName=jacket_side_08_l_capsule, center=<18.6118 0.0001 0>, rotation=<rot -90 -90 90.0002>}]
 jacket_side_08_r [{
   shapeType=Capsule
   length=30.75
   radius=9.25
   shapeName=jacket_side_08_r_capsule
   center=<18.4429 0 -1.8787>
   rotation=<rot -90 180 -179.9999>
  }]
 lowerarm_l []
 lowerarm_r []
 neck_02 []
 pelvis [{shapeType=Capsule, length=13.4028, radius=18.4095, center=<-3.9543 -1.6123 -0.1981>}]
 spine_01 [{shapeType=Capsule, length=9.2671, radius=13.2354, center=<-2.8183 -1.0873 0>}]
 spine_02 [
  {shapeType=Capsule, length=11.5, radius=11.9854, center=<1.179 -1.8646 0>}
  {shapeType=Capsule, length=20.6875, radius=9.3125, center=<-1.8317 2.606 0>, rotation=<rot 90 0 0>}
 ]
 spine_03 [
  {shapeType=Capsule, length=5, radius=12.7354, center=<-0.0747 0.1172 0>}
  {shapeType=Capsule, length=20.7188, radius=9.2812, center=<-1.2326 2.6059 0>, rotation=<rot 90 0 0>}
 ]
 spine_04 [
  {shapeType=Capsule, length=3.0796, radius=13.1416, center=<5.5182 -0.0712 0>}
  {shapeType=Capsule, length=10.0312, radius=9.2812, center=<-1.2326 1.308 0>, rotation=<rot 90 0 0>}
 ]
 spine_05 [{shapeType=Capsule, length=10.9859, radius=10.0791, center=<5.6377 0.3227 0>}]
 thigh_l []
 thigh_r []
 upperarm_l []
 upperarm_r []
}
constraints []
